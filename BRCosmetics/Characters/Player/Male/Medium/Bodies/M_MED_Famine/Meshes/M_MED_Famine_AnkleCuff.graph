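# Physics asset for the ankle cuff mesh of the male medium Famine player cosmetic, covering its dyn_ankle_cuff bones, apparently for secondary motion.
# Three box bodies, one per bone, form a chain from 01 through 02 to 03.
# Two constraints link neighbouring bodies with twist locked, and the swing limit angles are much tighter between 01 and 02 than between 02 and 03.
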PhysicsAsset /BRCosmetics/Characters/Player/Male/Medium/Bodies/M_MED_Famine/Meshes/M_MED_Famine_AnkleCuff
bodies {
 dyn_ankle_cuff_01 [{shapeType=Box, extentX=4, extentY=22.5, extentZ=2.5, center=<-2.8817 -0.048 -5.3475>, rotation=<rot 0 0 5.5885>}]
 dyn_ankle_cuff_02 [{shapeType=Box, extentX=15, extentY=15, extentZ=7, rotation=<rot 0 0 5.7975>}]
 dyn_ankle_cuff_03 [{shapeType=Box, extentX=6, extentY=6, extentZ=2, center=<3.7868 -0.0004 0>}]
}
constraints [
 {
  bone1Name=dyn_ankle_cuff_02
  bone2Name=dyn_ankle_cuff_01
  swing1LimitDegrees=5
  swing2LimitDegrees=5
  twistLimitDegrees=2
  twistMotion=Locked
 }
 {
  bone1Name=dyn_ankle_cuff_03
  bone2Name=dyn_ankle_cuff_02
  swing1LimitDegrees=45
  swing2LimitDegrees=45
  twistLimitDegrees=45
  twistMotion=Locked
 }
]
