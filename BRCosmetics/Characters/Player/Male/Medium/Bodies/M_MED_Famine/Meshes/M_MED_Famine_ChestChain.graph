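# Physics asset for the chest chain on the medium male Famine player cosmetic body.
# It gives box bodies to three two-bone dynamic chains: back left, back right and mid left, with small boxes on the 01 bones and larger ones on the 02 bones.
# A constraint joins each 02 bone to its 01 bone with swing and twist locked, so the configured angle limits allow no rotation between them.
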PhysicsAsset /BRCosmetics/Characters/Player/Male/Medium/Bodies/M_MED_Famine/Meshes/M_MED_Famine_ChestChain
bodies {
 dyn_chest_chain_back_l_01 [{shapeType=Box, extentX=1.01, extentY=1.01, extentZ=1.01, rotation=<rot 0 -90 0>}]
 dyn_chest_chain_back_l_02 [{shapeType=Box, extentX=4, extentY=4, extentZ=4}]
 dyn_chest_chain_back_r_01 [{shapeType=Box, extentX=1.01, extentY=1.01, extentZ=1.01, rotation=<rot 0 -90 0>}]
 dyn_chest_chain_back_r_02 [{shapeType=Box, extentX=4, extentY=4, extentZ=4}]
 dyn_chest_chain_mid_l_01 [{shapeType=Box, extentX=1.01, extentY=1.01, extentZ=1.01, rotation=<rot 0 -90 0>}]
 dyn_chest_chain_mid_l_02 [{shapeType=Box, extentX=4, extentY=4, extentZ=4}]
}
constraints [
 {
  bone1Name=dyn_chest_chain_back_r_02
  bone2Name=dyn_chest_chain_back_r_01
  swing1LimitDegrees=45
  swing1Motion=Locked
  swing2LimitDegrees=45
  swing2Motion=Locked
  twistLimitDegrees=45
  twistMotion=Locked
 }
 {
  bone1Name=dyn_chest_chain_back_l_02
  bone2Name=dyn_chest_chain_back_l_01
  swing1LimitDegrees=45
  swing1Motion=Locked
  swing2LimitDegrees=45
  swing2Motion=Locked
  twistLimitDegrees=45
  twistMotion=Locked
 }
 {
  bone1Name=dyn_chest_chain_mid_l_02
  bone2Name=dyn_chest_chain_mid_l_01
  swing1LimitDegrees=45
  swing1Motion=Locked
  swing2LimitDegrees=45
  swing2Motion=Locked
  twistLimitDegrees=45
  twistMotion=Locked
 }
]
