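# Control Rig for the Famine male medium body cosmetic that offsets a knee brace bone procedurally as the left knee bends.
# The left calf's rotation relative to the thigh becomes an Euler Z angle, which a clamped remap turns into an additive local Z translation on dyn_knee_brace_01.
# A duplicate branch targets dyn_ankle_cuff_01 on the right calf, also reading the left knee, but has no execution link.
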ControlRigBlueprint /BRCosmetics/Characters/Player/Male/Medium/Bodies/M_MED_Famine/Meshes/M_MED_Famine_CtrlRig
graphs [
 RigVMFunctionLibrary {links [], nodes []}
 RigVMModel {nodes [
   RigUnit_BeginExecution = &RigUnit_BeginExecution()
   n1 = &RigUnit_GetRelativeBoneTransform(Bone=calf_l, Space=thigh_l)
   /Script/ControlRig.RigUnit_ModifyBoneTransforms(
    BoneToModify=((Bone="dyn_knee_brace_01",Transform=(Rotation=(X=0,Y=0,Z=0,W=1),Translation=(X=0,Y=0,Z=5),Scale3D=(X=1,Y=1,Z=1))))
    Mode=AdditiveLocal
    Weight=1
    WeightMaximum=1
    WeightMinimum=0
    BoneToModify.0.Transform.Translation.Z=&RigVMFunction_MathFloatRemap(
     SourceMaximum=115
     SourceMinimum=24
     TargetMaximum=6
     bClamp=true
     Value=&RigVMFunction_MathQuaternionToEuler(RotationOrder=ZXY, Value=@n1.Transform.Rotation).Result.Z
    ).Result
    ExecutePin=@RigUnit_BeginExecution.ExecutePin
   ) {is=struct RigVMUnitNode}
   n2 = &RigUnit_GetRelativeBoneTransform(Bone=calf_l, Space=thigh_l)
   &RigUnit_ModifyBoneTransforms(
    BoneToModify=((Bone="dyn_ankle_cuff_01",Transform=(Rotation=(X=0,Y=0,Z=0,W=1),Translation=(X=0,Y=0,Z=5),Scale3D=(X=1,Y=1,Z=1))))
    ExecutePin=()
    Mode=AdditiveLocal
    Weight=1
    WeightMaximum=1
    WeightMinimum=0
    BoneToModify.0.Transform.Translation.Z=&RigVMFunction_MathFloatRemap(
     SourceMaximum=115
     SourceMinimum=24
     TargetMaximum=6
     TargetMinimum=3
     bClamp=true
     Value=&RigVMFunction_MathQuaternionToEuler(RotationOrder=ZXY, Value=@n2.Transform.Rotation).Result.Z
    ).Result
   )
  ]}
]
hierarchy [
 root {loc=<0 0 0>, rot=<0 0 0 1>}
 pelvis {parent=root, loc=<0 -0.7071 93.1191>, rot=<0.0127 0.707 -0.0127 -0.707>}
 spine_01 {parent=pelvis, loc=<6.7494 -0.6388 0>, rot=<0 0 0.034 -0.9994>}
 spine_02 {parent=spine_01, loc=<7.2518 -0.035 0>, rot=<0 0 0.0238 -0.9997>}
 spine_03 {parent=spine_02, loc=<6.0104 0 0>, rot=<0 0 0.0149 -0.9999>}
 spine_04 {parent=spine_03, loc=<8.1024 0 0>, rot=<0 0 -0.0132 -0.9999>}
 spine_05 {parent=spine_04, loc=<10.0146 0 0>, rot=<0 0 -0.0015 -1>}
 clavicle_l {parent=spine_05, loc=<4.3575 -0.1466 -2.3223>, rot=<0.0465 -0.7289 -0.009 -0.6829>}
 upperarm_l {parent=clavicle_l, loc=<15.5111 0 0>, rot=<0.0226 -0.3234 -0.0088 -0.9459>}
 lowerarm_l {parent=upperarm_l, loc=<26.6514 0 0>, rot=<0.0391 -0.0099 -0.2453 -0.9686>}
 hand_l {parent=lowerarm_l, loc=<22.3491 0 0>, rot=<-0.7081 -0.0507 0.0975 -0.6975>}
 index_metacarpal_l {parent=hand_l, loc=<1.6535 0.0183 -2.96>, rot=<0.0158 -0.0124 -0.0173 -0.9996>}
 index_01_l {parent=index_metacarpal_l, loc=<8.3008 0.0357 -0.0524>, rot=<-0.0044 0.0485 0.0924 -0.9945>}
 index_02_l {parent=index_01_l, loc=<3.3587 -0.3004 0.0452>, rot=<-0.002 0.0027 0.3225 -0.9466>}
 index_03_l {parent=index_02_l, loc=<2.5578 -0.123 -0.0252>, rot=<-0.002 0.0067 0.2558 -0.9667>}
 middle_metacarpal_l {parent=hand_l, loc=<2.1856 -0.1725 -0.5984>, rot=<0.006 0.0168 -0.0329 -0.9993>}
 middle_01_l {parent=middle_metacarpal_l, loc=<7.3275 -0.2447 0.0129>, rot=<-0.0208 0.0042 0.0858 -0.9961>}
 middle_02_l {parent=middle_01_l, loc=<4.0123 0.1401 0.1175>, rot=<-0.0218 0.0258 0.3512 -0.9357>}
 middle_03_l {parent=middle_02_l, loc=<2.6077 0.0534 0.1373>, rot=<0.0082 0.0188 0.1708 -0.9851>}
 pinky_metacarpal_l {parent=hand_l, loc=<1.7616 -0.9753 3.3814>, rot=<-0.0397 0.0814 -0.069 -0.9935>}
 pinky_01_l {parent=pinky_metacarpal_l, loc=<7.1709 -0.4831 -0.1094>, rot=<-0.0654 -0.0035 0.2242 -0.9723>}
 pinky_02_l {parent=pinky_01_l, loc=<2.8274 0.0893 -0.1043>, rot=<0.0037 0.0011 0.2854 -0.9584>}
 pinky_03_l {parent=pinky_02_l, loc=<2.2046 0.0307 -0.0078>, rot=<-0.0002 0.0098 0.1445 -0.9895>}
 ring_metacarpal_l {parent=hand_l, loc=<2.2074 -0.3625 1.3638>, rot=<-0.0284 0.0722 -0.0458 -0.9959>}
 ring_01_l {parent=ring_metacarpal_l, loc=<7.0295 -0.6197 -0.1446>, rot=<-0.0489 -0.0252 0.1892 -0.9804>}
 ring_02_l {parent=ring_01_l, loc=<3.6174 0.4646 -0.1027>, rot=<0 0 0.2806 -0.9598>}
 ring_03_l {parent=ring_02_l, loc=<2.4152 0.1141 0.0199>, rot=<0 0 0.1189 -0.9929>}
 thumb_01_l {parent=hand_l, loc=<1.3118 -0.3191 -2.459>, rot=<0.6546 -0.4042 0.0664 -0.6354>}
 thumb_02_l {parent=thumb_01_l, loc=<5.9336 -0.0448 0.0401>, rot=<-0.0037 0.0213 0.298 -0.9543>}
 thumb_03_l {parent=thumb_02_l, loc=<2.2223 0.3682 -0.0907>, rot=<0.0118 -0.013 0.0066 -0.9998>}
 lowerarm_twist_01_l {parent=lowerarm_l, loc=<17.8549 0 0>, rot=<0 0 0 1>}
 lowerarm_twist_02_l {parent=lowerarm_l, loc=<9.6199 0 0>, rot=<0 0 0 1>}
 upperarm_twist_01_l {parent=upperarm_l, loc=<0.0041 0 0>, rot=<0 0 0 -1>}
 upperarm_twist_02_l {parent=upperarm_l, loc=<13.7941 0 0>, rot=<0 0 0 -1>}
 clavicle_r {parent=spine_05, loc=<4.3575 -0.1466 2.3223>, rot=<0.7289 0.0465 -0.6829 0.009>}
 upperarm_r {parent=clavicle_r, loc=<-15.5111 0 0>, rot=<0.0226 -0.3234 -0.0088 -0.9459>}
 lowerarm_r {parent=upperarm_r, loc=<-26.6514 0 0>, rot=<0.0391 -0.0099 -0.2453 -0.9686>}
 hand_r {parent=lowerarm_r, loc=<-22.3492 0 0>, rot=<-0.7081 -0.0507 0.0975 -0.6975>}
 index_metacarpal_r {parent=hand_r, loc=<-1.6534 -0.0181 2.9601>, rot=<0.0158 -0.0124 -0.0173 -0.9996>}
 index_01_r {parent=index_metacarpal_r, loc=<-8.3008 -0.0357 0.0524>, rot=<-0.0044 0.0485 0.0924 -0.9945>}
 index_02_r {parent=index_01_r, loc=<-3.3589 0.3003 -0.0453>, rot=<-0.002 0.0027 0.3225 -0.9466>}
 index_03_r {parent=index_02_r, loc=<-2.5578 0.1231 0.0251>, rot=<-0.002 0.0067 0.2558 -0.9667>}
 middle_metacarpal_r {parent=hand_r, loc=<-2.1856 0.1726 0.5984>, rot=<0.006 0.0168 -0.0329 -0.9993>}
 middle_01_r {parent=middle_metacarpal_r, loc=<-7.3273 0.2447 -0.0129>, rot=<-0.0208 0.0042 0.0858 -0.9961>}
 middle_02_r {parent=middle_01_r, loc=<-4.0142 -0.1411 -0.07>, rot=<-0.0217 0.0259 0.3512 -0.9357>}
 middle_03_r {parent=middle_02_r, loc=<-2.6074 -0.0534 -0.1312>, rot=<0.0083 0.0188 0.1708 -0.9851>}
 pinky_metacarpal_r {parent=hand_r, loc=<-1.7617 0.9753 -3.3814>, rot=<-0.0397 0.0814 -0.069 -0.9935>}
 pinky_01_r {parent=pinky_metacarpal_r, loc=<-7.1707 0.4832 0.1094>, rot=<-0.0654 -0.0035 0.2242 -0.9723>}
 pinky_02_r {parent=pinky_01_r, loc=<-2.8274 -0.0892 0.1043>, rot=<0.0037 0.0011 0.2854 -0.9584>}
 pinky_03_r {parent=pinky_02_r, loc=<-2.2045 -0.031 0.0078>, rot=<-0.0002 0.0098 0.1445 -0.9895>}
 ring_metacarpal_r {parent=hand_r, loc=<-2.2073 0.3627 -1.3638>, rot=<-0.0284 0.0722 -0.0458 -0.9959>}
 ring_01_r {parent=ring_metacarpal_r, loc=<-7.0297 0.6197 0.1446>, rot=<-0.0489 -0.0252 0.1892 -0.9804>}
 ring_02_r {parent=ring_01_r, loc=<-3.6172 -0.4648 0.1027>, rot=<0 0 0.2806 -0.9598>}
 ring_03_r {parent=ring_02_r, loc=<-2.4152 -0.1141 -0.0199>, rot=<0 0 0.1189 -0.9929>}
 thumb_01_r {parent=hand_r, loc=<-1.3119 0.3192 2.459>, rot=<0.6546 -0.4042 0.0664 -0.6354>}
 thumb_02_r {parent=thumb_01_r, loc=<-5.9335 0.0447 -0.0401>, rot=<-0.0037 0.0213 0.298 -0.9543>}
 thumb_03_r {parent=thumb_02_r, loc=<-2.2223 -0.3682 0.0907>, rot=<0.0118 -0.013 0.0066 -0.9998>}
 lowerarm_twist_01_r {parent=lowerarm_r, loc=<-17.855 0 0>, rot=<0 0 0 -1>}
 lowerarm_twist_02_r {parent=lowerarm_r, loc=<-9.6199 0 0>, rot=<0 0 0 -1>}
 upperarm_twist_01_r {parent=upperarm_r, loc=<-0.0041 0 0>, rot=<0 0 0 1>}
 upperarm_twist_02_r {parent=upperarm_r, loc=<-13.7942 0 0>, rot=<0 0 0 1>}
 dyn_arm_chain_master {parent=upperarm_r, loc=<-10.9941 0 0>, rot=<0 0 0 1>}
 dyn_arm_chain_top_01 {parent=dyn_arm_chain_master, loc=<0 -0.041 -3.828>, rot=<0 0 0 1>}
 dyn_arm_chain_top_02 {parent=dyn_arm_chain_top_01, loc=<0 0 0>, rot=<0 0 0 1>}
 dyn_arm_chain_bottom_01 {parent=dyn_arm_chain_master, loc=<0 2.7505 6.2707>, rot=<0 0 0 1>}
 dyn_arm_chain_bottom_02 {parent=dyn_arm_chain_bottom_01, loc=<0 0 0>, rot=<0 0 0 1>}
 dyn_arm_chain_back_01 {parent=dyn_arm_chain_master, loc=<0 7.0599 0.8105>, rot=<0 0 0 1>}
 dyn_arm_chain_back_02 {parent=dyn_arm_chain_back_01, loc=<0 0 0>, rot=<0 0 0 1>}
 dyn_arm_chain_front_01 {parent=dyn_arm_chain_master, loc=<0.2986 -4.9806 2.7648>, rot=<0 0 0 1>}
 dyn_arm_chain_front_02 {parent=dyn_arm_chain_front_01, loc=<0 0 0>, rot=<0 0 0 1>}
 dyn_tatters_master {parent=upperarm_r, loc=<-18.678 0 0>, rot=<0 0 0 1>}
 dyn_tatter_back_01 {parent=dyn_tatters_master, loc=<3.5333 2.9828 -2.2013>, rot=<-0.1708 -0.0066 0.9845 -0.0382>}
 dyn_tatter_back_02 {parent=dyn_tatter_back_01, loc=<4.9902 0 0>, rot=<0.0017 0.1743 0.034 -0.9841>}
 dyn_tatter_back_03 {parent=dyn_tatter_back_02, loc=<7.2222 0 0>, rot=<0.0201 -0.1712 0.0907 -0.9808>}
 dyn_tatter_top_01 {parent=dyn_tatters_master, loc=<3.3544 -1.0174 -2.8238>, rot=<-0.2232 0.0116 0.9747 -0.0032>}
 dyn_tatter_top_02 {parent=dyn_tatter_top_01, loc=<4.3616 0 0>, rot=<0.0086 0.1405 -0.1262 -0.982>}
 dyn_tatter_top_03 {parent=dyn_tatter_top_02, loc=<3.7083 0 0>, rot=<0.0032 0.0967 0.1543 -0.9833>}
 dyn_tatter_front {parent=dyn_tatters_master, loc=<2.7433 -3.9863 1.3205>, rot=<0.0564 -0.0073 0.9979 -0.0322>}
 dyn_tatter_bottom {parent=dyn_tatters_master, loc=<2.0051 -1.0905 5.8201>, rot=<0.1265 -0.0157 0.9819 -0.14>}
 dyn_shoulder_chain_master {parent=clavicle_r, loc=<-15.5111 0 0>, rot=<0.0226 -0.3234 -0.0088 -0.9459>}
 dyn_shoulder_chain_top_01 {parent=dyn_shoulder_chain_master, loc=<7.7578 0.4662 -2.3207>, rot=<0.014 -0.9345 -0.0134 -0.3555>}
 dyn_shoulder_chain_top_02 {parent=dyn_shoulder_chain_top_01, loc=<0 0 0>, rot=<0 0 0 1>}
 dyn_shoulder_chain_bottom_01 {parent=dyn_shoulder_chain_master, loc=<-8.646 0.422 9.254>, rot=<0.014 -0.9345 -0.0134 -0.3555>}
 dyn_shoulder_chain_bottom_02 {parent=dyn_shoulder_chain_bottom_01, loc=<0 0 0>, rot=<0 0 0 1>}
 dyn_shoulder_chain_back_01 {parent=dyn_shoulder_chain_master, loc=<-3.0621 9.4847 7.1766>, rot=<0.014 -0.9345 -0.0134 -0.3555>}
 dyn_shoulder_chain_back_02 {parent=dyn_shoulder_chain_back_01, loc=<0 0.1084 -0.1828>, rot=<0 0 0 1>}
 dyn_shoulder_chain_back_03 {parent=dyn_shoulder_chain_back_02, loc=<0 -0.1084 0.1828>, rot=<-0.4851 -0.5194 0.4871 -0.5076>}
 dyn_shoulder_chain_back_04 {parent=dyn_shoulder_chain_back_03, loc=<9.4356 0 0>, rot=<0.0001 -0.0148 -0.0047 -0.9999>}
 dyn_shoulder_chain_back_05 {parent=dyn_shoulder_chain_back_04, loc=<8.4923 0 0>, rot=<0 0.0078 0 -1>}
 dyn_shoulder_chain_back_06 {parent=dyn_shoulder_chain_back_05, loc=<8.1149 0 0>, rot=<0 0 0 1>}
 dyn_shoulder_chain_front_01 {parent=dyn_shoulder_chain_master, loc=<-0.3 -8.124 2.873>, rot=<0.014 -0.9345 -0.0134 -0.3555>}
 dyn_shoulder_chain_front_02 {parent=dyn_shoulder_chain_front_01, loc=<0 0 0>, rot=<0 0 0 1>}
 neck_01 {parent=spine_05, loc=<11.3138 0.1529 0>, rot=<0 0 -0.2003 -0.9797>}
 neck_02 {parent=neck_01, loc=<4.5627 0 0>, rot=<0 0 0.0504 -0.9987>}
 head {parent=neck_02, loc=<5.3914 0 0>, rot=<0 0 0.115 -0.9934>}
 dyn_zipper_front_r_01 {parent=spine_05, loc=<-7.8509 10.2234 8.4444>, rot=<0.0024 -0.8761 0.4772 0.069>}
 dyn_zipper_front_r_02 {parent=dyn_zipper_front_r_01, loc=<0 0 0>, rot=<0 0 0 -1>}
 dyn_zipper_front_r_03 {parent=spine_05, loc=<-12.1812 8.6364 9.9847>, rot=<0.0024 -0.8761 0.4772 0.069>}
 dyn_zipper_front_r_04 {parent=dyn_zipper_front_r_03, loc=<0 0 0>, rot=<0 0 0 -1>}
 dyn_zipper_front_r_05 {parent=spine_05, loc=<-21.254 8.0279 9.1955>, rot=<0.011 -0.8102 0.5821 0.0682>}
 dyn_zipper_front_r_06 {parent=dyn_zipper_front_r_05, loc=<0 0 0>, rot=<0 0 0 1>}
 dyn_zipper_front_r_07 {parent=spine_05, loc=<-27.1517 6.6149 11.2769>, rot=<0.011 -0.8102 0.5821 0.0682>}
 dyn_zipper_front_r_08 {parent=dyn_zipper_front_r_07, loc=<0 0 0>, rot=<0 0 0 1>}
 dyn_zipper_front_l_01 {parent=spine_05, loc=<-7.7149 9.9234 -8.286>, rot=<0.004 -0.4733 0.8799 -0.0413>}
 dyn_zipper_front_l_02 {parent=dyn_zipper_front_l_01, loc=<0 0 0>, rot=<0 0 0 1>}
 dyn_zipper_front_l_03 {parent=spine_05, loc=<-13.0882 9.1844 -9.3834>, rot=<0.004 -0.4733 0.8799 -0.0413>}
 dyn_zipper_front_l_04 {parent=dyn_zipper_front_l_03, loc=<0 0 0>, rot=<0 0 0 1>}
 dyn_zipper_front_l_07 {parent=spine_05, loc=<-26.6339 7.5733 -10.0306>, rot=<-0.0009 -0.3644 0.9303 -0.0415>}
 dyn_zipper_front_l_08 {parent=dyn_zipper_front_l_07, loc=<0 0 0>, rot=<0 0 0 -1>}
 dyn_zipper_front_l_05 {parent=spine_05, loc=<-23.7522 4.5187 -12.9688>, rot=<-0.0009 -0.3644 0.9303 -0.0415>}
 dyn_zipper_front_l_06 {parent=dyn_zipper_front_l_05, loc=<0 0 0>, rot=<0 0 0 -1>}
 dyn_zipper_back_r_01 {parent=spine_05, loc=<-24.8413 -8.7859 12.2032>, rot=<0.1276 -0.9186 -0.3276 0.1806>}
 dyn_zipper_back_r_02 {parent=dyn_zipper_back_r_01, loc=<0 0 0>, rot=<0 0 0 -1>}
 dyn_zipper_back_l_03 {parent=spine_05, loc=<-28.7814 -9.4675 -12.38>, rot=<-0.0473 0.4477 0.8922 -0.0375>}
 dyn_zipper_back_l_04 {parent=dyn_zipper_back_l_03, loc=<0 -0.1452 0>, rot=<0 0 0 1>}
 dyn_zipper_back_l_01 {parent=spine_05, loc=<-23.7016 -8.9685 -10.8012>, rot=<-0.0432 0.3517 0.9341 -0.0422>}
 dyn_zipper_back_l_02 {parent=dyn_zipper_back_l_01, loc=<0 0 0>, rot=<0 0 0 -1>}
 dyn_zipper_back_r_03 {parent=spine_05, loc=<-28.0602 -10.8362 11.4509>, rot=<0.162 -0.8341 -0.5024 0.1599>}
 dyn_zipper_back_r_04 {parent=dyn_zipper_back_r_03, loc=<0 0 0>, rot=<0 0 0 1>}
 dyn_zipper_cuff_l_01 {parent=spine_05, loc=<-26.8852 0.8791 -49.4264>, rot=<0.1344 -0.7496 -0.6454 0.0595>}
 dyn_zipper_cuff_l_02 {parent=dyn_zipper_cuff_l_01, loc=<0 0 0>, rot=<0 0 0 -1>}
 dyn_chest_chain_mid_r_01 {parent=spine_05, loc=<-13.2264 -1.1984 14.9496>, rot=<0.0284 -0.7065 -0.0284 -0.7065>}
 dyn_chest_chain_mid_r_02 {parent=dyn_chest_chain_mid_r_01, loc=<0 0 0>, rot=<0 0 0 1>}
 dyn_chest_chain_front_r_01 {parent=spine_05, loc=<-6.5612 7.9434 10.8155>, rot=<0.0284 -0.7065 -0.0284 -0.7065>}
 dyn_chest_chain_front_r_02 {parent=dyn_chest_chain_front_r_01, loc=<0 0 0>, rot=<0 0 0 1>}
 dyn_chest_chain_back_r_01 {parent=spine_05, loc=<-3.6267 -10.8133 9.257>, rot=<0.0284 -0.7065 -0.0284 -0.7065>}
 dyn_chest_chain_back_r_02 {parent=dyn_chest_chain_back_r_01, loc=<0 0 0>, rot=<0 0 0 1>}
 dyn_zipper_front_r_09 {parent=spine_05, loc=<-36.6769 7.0086 9.179>, rot=<0.0061 -0.8492 0.5235 0.0688>}
 dyn_zipper_front_r_10 {parent=dyn_zipper_front_r_09, loc=<0 0 0>, rot=<0 0 0 1>}
 dyn_chest_chain_back_l_01 {parent=spine_05, loc=<-3.5314 -12.0042 -9.4712>, rot=<0.7065 0.0284 -0.7065 0.0284>}
 dyn_chest_chain_back_l_02 {parent=dyn_chest_chain_back_l_01, loc=<0 0 0>, rot=<0 0 0 1>}
 dyn_chest_chain_mid_l_01 {parent=spine_05, loc=<-9.589 -3.5566 -14.9496>, rot=<0.7065 0.0284 -0.7065 0.0284>}
 dyn_chest_chain_mid_l_02 {parent=dyn_chest_chain_mid_l_01, loc=<0 0 0>, rot=<0 0 0 1>}
 dyn_chest_chain_front_l_01 {parent=spine_05, loc=<-4.0482 8.0619 -12.1068>, rot=<0.7065 0.0284 -0.7065 0.0284>}
 dyn_chest_chain_front_l_02 {parent=dyn_chest_chain_front_l_01, loc=<0 0 0>, rot=<0 0 0 1>}
 thigh_l {parent=pelvis, loc=<-2.5693 -0.0109 -7.7501>, rot=<0.0012 0.0572 0.0155 -0.9982>}
 calf_l {parent=thigh_l, loc=<-39.0484 0 0>, rot=<0.0006 -0.0161 -0.0391 -0.9991>}
 calf_twist_01_l {parent=calf_l, loc=<-35.8131 -0.0967 0.0397>, rot=<0.0237 0.0095 -0.0035 -0.9997>}
 calf_twist_02_l {parent=calf_l, loc=<-18.8064 -0.0967 0.0397>, rot=<0.0237 0.0095 -0.0035 -0.9997>}
 foot_l {parent=calf_l, loc=<-42.561 0.1077 0.4292>, rot=<0.0418 -0.0455 0.0343 -0.9975>}
 ball_l {parent=foot_l, loc=<-6.1599 13.8264 0.6806>, rot=<0.0001 -0.0001 -0.7063 -0.7079>}
 shoelace_l_01 {parent=foot_l, loc=<1.9305 5.7504 0.519>, rot=<0.1282 0.6107 -0.3554 -0.6959>}
 shoelace_l_02 {parent=shoelace_l_01, loc=<0 0 0>, rot=<0 0 0 -1>}
 shoelace_l_04 {parent=foot_l, loc=<2.2828 5.3843 -0.908>, rot=<0.1884 -0.8162 -0.3032 -0.4543>}
 shoelace_l_05 {parent=shoelace_l_04, loc=<0 0 0>, rot=<0 0 0 -1>}
 thigh_twist_01_l {parent=thigh_l, loc=<-21.6303 0 0>, rot=<0.0013 0 0 -1>}
 dyn_thigh_tab_05 {parent=thigh_l, loc=<-11.6206 1.0968 -6.1228>, rot=<-0.2526 -0.5955 -0.2281 -0.7277>}
 dyn_thigh_tab_06 {parent=dyn_thigh_tab_05, loc=<0 0 0>, rot=<0 0 0 1>}
 dyn_knee_brace_01 {parent=thigh_l, loc=<-28.9592 -0.055 -0.2754>, rot=<-0.3416 -0.6645 -0.4507 -0.4884>}
 dyn_knee_brace_02 {parent=dyn_knee_brace_01, loc=<0 0 0>, rot=<0 0 0 1>}
 thigh_r {parent=pelvis, loc=<-2.5691 -0.0109 7.7501>, rot=<0.0572 -0.0012 0.9982 0.0155>}
 calf_r {parent=thigh_r, loc=<39.0487 0 0>, rot=<0.0006 -0.0161 -0.0391 -0.9991>}
 calf_twist_01_r {parent=calf_r, loc=<35.8132 0.0967 -0.0397>, rot=<0.0237 0.0095 -0.0035 -0.9997>}
 calf_twist_02_r {parent=calf_r, loc=<18.8064 0.0967 -0.0397>, rot=<0.0237 0.0095 -0.0035 -0.9997>}
 foot_r {parent=calf_r, loc=<42.5611 -0.1077 -0.4292>, rot=<0.0418 -0.0455 0.0343 -0.9975>}
 ball_r {parent=foot_r, loc=<6.1598 -13.8264 -0.6806>, rot=<0.0001 -0.0001 -0.7063 -0.7079>}
 shoelace_r_03 {parent=foot_r, loc=<-1.9412 -6.0372 1.5564>, rot=<0.2783 0.4874 0.1781 -0.8082>}
 shoelace_r_04 {parent=shoelace_r_03, loc=<0 0 0>, rot=<0 0 0 1>}
 shoelace_r_01 {parent=foot_r, loc=<-2.0121 -5.9234 -0.9683>, rot=<0.8028 -0.1557 -0.5387 -0.2026>}
 shoelace_r_02 {parent=shoelace_r_01, loc=<0 0 0>, rot=<0 0 0 1>}
 dyn_ankle_cuff_01 {parent=calf_r, loc=<38.2929 0.7814 -0.2579>, rot=<0.4072 -0.5987 -0.5582 0.4052>}
 dyn_ankle_cuff_02 {parent=dyn_ankle_cuff_01, loc=<0 0 0>, rot=<0 0 0 -1>}
 dyn_ankle_cuff_03 {parent=dyn_ankle_cuff_02, loc=<-8.5395 1.0944 -0.3768>, rot=<0.0582 -0.8083 0.0421 -0.5844>}
 thigh_twist_01_r {parent=thigh_r, loc=<21.6303 0 0>, rot=<0.0013 0 0 -1>}
 dyn_thigh_tab_01 {parent=thigh_r, loc=<10.5612 -3.3925 5.6839>, rot=<-0.1959 0.6363 0.3382 -0.6651>}
 dyn_thigh_tab_02 {parent=dyn_thigh_tab_01, loc=<0 0 0>, rot=<0 0 0 -1>}
 dyn_thigh_tab_03 {parent=thigh_r, loc=<16.403 -4.2017 5.1104>, rot=<-0.3908 0.5551 0.4489 -0.5812>}
 dyn_thigh_tab_04 {parent=dyn_thigh_tab_03, loc=<0 0 0>, rot=<0 0 0 1>}
 dyn_jacket_f_l_01 {parent=pelvis, loc=<13.6806 7.4757 -6.6664>, rot=<0.0047 0.2608 0.9652 0.0173>}
 dyn_jacket_f_l_02 {parent=dyn_jacket_f_l_01, loc=<0 0 0>, rot=<0 0 0 -1>}
 dyn_jacket_f_l_03 {parent=dyn_jacket_f_l_02, loc=<6.7781 0 0>, rot=<0 0 0 -1>}
 dyn_jacket_l_01 {parent=pelvis, loc=<12.5868 -2.3597 -14.1347>, rot=<-0.0137 -0.7627 -0.6465 -0.0116>}
 dyn_jacket_l_02 {parent=dyn_jacket_l_01, loc=<0 0 0>, rot=<0 0 0 -1>}
 dyn_jacket_l_03 {parent=dyn_jacket_l_02, loc=<6.7781 0 0>, rot=<0 0 0 -1>}
 dyn_jacket_r_01 {parent=pelvis, loc=<12.6051 -1.8471 15.6466>, rot=<-0.0129 -0.7214 0.6923 0.0124>}
 dyn_jacket_r_02 {parent=dyn_jacket_r_01, loc=<0 0 0>, rot=<0 0 0 -1>}
 dyn_jacket_r_03 {parent=dyn_jacket_r_02, loc=<6.7781 0 0>, rot=<0 0 0 -1>}
 dyn_jacket_b_l_01 {parent=pelvis, loc=<12.2481 -11.8075 -6.0083>, rot=<-0.0177 -0.9857 -0.1676 -0.003>}
 dyn_jacket_b_l_02 {parent=dyn_jacket_b_l_01, loc=<0 0 0>, rot=<0 0 0 -1>}
 dyn_jacket_b_l_03 {parent=dyn_jacket_b_l_02, loc=<6.7781 0 0>, rot=<0 0 0 -1>}
 dyn_jacket_b_r_01 {parent=pelvis, loc=<12.2568 -11.5654 7.1222>, rot=<-0.011 -0.6162 0.7874 0.0141>}
 dyn_jacket_b_r_02 {parent=dyn_jacket_b_r_01, loc=<0 0 0>, rot=<0 0 0 -1>}
 dyn_jacket_b_r_03 {parent=dyn_jacket_b_r_02, loc=<6.7781 0 0>, rot=<0 0.5663 0 -0.8242>}
 dyn_jacket_f_r_01 {parent=pelvis, loc=<12.8691 5.5161 10.3624>, rot=<-0.0052 -0.2897 0.9569 0.0171>}
 dyn_jacket_f_r_02 {parent=dyn_jacket_f_r_01, loc=<0 0 0>, rot=<0 0 0 -1>}
 dyn_jacket_f_r_03 {parent=dyn_jacket_f_r_02, loc=<6.7781 0 0>, rot=<0 0 0 -1>}
]
