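# Physics asset for the jacket of the male medium Famine player cosmetic, giving its dangling panels physics bodies for secondary motion.
# Six two-bone chains cover front, back and side on left and right, each with a small root box and a larger offset tip box.
# Constraints join only tip to root within each chain, with per-panel swing angles and locked twist on the front pair.
# The pelvis carries two side-by-side capsules and has no constraints.
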PhysicsAsset /BRCosmetics/Characters/Player/Male/Medium/Bodies/M_MED_Famine/Meshes/M_MED_Famine_Jacket
bodies {
 dyn_jacket_b_l_01 [{shapeType=Box, extentX=4, extentY=4, extentZ=4}]
 dyn_jacket_b_l_02 [{shapeType=Box, extentX=7.5, extentY=4, extentZ=10.5, center=<3.6996 -0.0001 0>}]
 dyn_jacket_b_r_01 [{shapeType=Box, extentX=4, extentY=4, extentZ=4}]
 dyn_jacket_b_r_02 [{shapeType=Box, extentX=9, extentY=11.5, extentZ=4, center=<4.5661 -0.0001 0>}]
 dyn_jacket_f_l_01 [{shapeType=Box, extentX=4, extentY=4, extentZ=4}]
 dyn_jacket_f_l_02 [{shapeType=Box, extentX=13, extentY=4, extentZ=6.5, center=<8.7204 -0.0001 0>}]
 dyn_jacket_f_r_01 [{shapeType=Box, extentX=4, extentY=4, extentZ=4}]
 dyn_jacket_f_r_02 [{shapeType=Box, extentX=9.5, extentY=5.5, extentZ=4, center=<7.9026 -0.0002 0>}]
 dyn_jacket_l_01 [{shapeType=Box, extentX=4, extentY=4, extentZ=4}]
 dyn_jacket_l_02 [{shapeType=Box, extentX=9.5, extentY=4, extentZ=10, center=<5.1608 0 0.0001>}]
 dyn_jacket_r_01 [{shapeType=Box, extentX=4, extentY=4, extentZ=4}]
 dyn_jacket_r_02 [{shapeType=Box, extentX=8, extentY=4, extentZ=12, center=<4.3857 0.0001 0>}]
 pelvis [
  {shapeType=Capsule, length=34.5, radius=6.25, center=<11.9912 -2.3296 6.3331>, rotation=<rot -90 0 0>}
  {shapeType=Capsule, length=34.25, radius=6.5, center=<11.9912 -2.3629 -5.1329>, rotation=<rot -90 0 0>}
 ]
}
constraints [
 {
  bone1Name=dyn_jacket_f_l_02
  bone2Name=dyn_jacket_f_l_01
  swing1LimitDegrees=40
  swing2LimitDegrees=10
  twistLimitDegrees=45
  twistMotion=Locked
 }
 {bone1Name=dyn_jacket_b_l_02, bone2Name=dyn_jacket_b_l_01, swing1LimitDegrees=30, swing2LimitDegrees=20, twistLimitDegrees=2}
 {bone1Name=dyn_jacket_l_02, bone2Name=dyn_jacket_l_01, swing1LimitDegrees=30, swing2LimitDegrees=10, twistLimitDegrees=2}
 {bone1Name=dyn_jacket_b_r_02, bone2Name=dyn_jacket_b_r_01, swing1LimitDegrees=20, swing2LimitDegrees=20, twistLimitDegrees=2}
 {bone1Name=dyn_jacket_r_02, bone2Name=dyn_jacket_r_01, swing1LimitDegrees=30, swing2LimitDegrees=10, twistLimitDegrees=2}
 {
  bone1Name=dyn_jacket_f_r_02
  bone2Name=dyn_jacket_f_r_01
  swing1LimitDegrees=40
  swing2LimitDegrees=10
  twistLimitDegrees=45
  twistMotion=Locked
 }
]
