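# Physics asset for the sleeve tatters on the medium male Famine body, set up for bone-driven secondary motion at the right forearm.
# Box bodies form a three-link top chain, a four-link back chain and single front and bottom flaps, all constrained to a capsule master body.
# An unconstrained lowerarm_r capsule adds the arm's shape.
# Joints use tight angular limits, and the back chain's root locks twist and one swing axis.
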PhysicsAsset /BRCosmetics/Characters/Player/Male/Medium/Bodies/M_MED_Famine/Meshes/M_MED_Famine_SleeveTatters
bodies {
 dyn_tatter_back_01 [{shapeType=Box, extentX=1.0701, extentY=3.6192, extentZ=5.6277, center=<0.9767 -0.0002 1.9076>}]
 dyn_tatter_back_02 [{shapeType=Box, extentX=1.1875, extentY=8, extentZ=10.8377, center=<-0.0002 0 3.3073>}]
 dyn_tatter_back_03 [{shapeType=Box, extentX=2.4381, extentY=3.0175, extentZ=4.9136, center=<-0.4696 0.0001 1.6796>}]
 dyn_tatter_back_04 [{shapeType=Box, extentX=2.0857, extentY=1.6537, extentZ=4.5068, center=<0 0 1.9462>}]
 dyn_tatter_bottom [{shapeType=Box, extentX=7.5, extentY=4, extentZ=2.5, center=<1.6587 0 0>}]
 dyn_tatter_front [{shapeType=Box, extentX=6.5, extentY=1, extentZ=5, center=<2.362 0 0>}]
 dyn_tatter_top_01 [{shapeType=Box, extentX=4, extentY=3.5, extentZ=1, center=<1.4902 0.0001 0>}]
 dyn_tatter_top_02 [{shapeType=Box, extentX=4, extentY=4.5, extentZ=1, center=<1.2324 -0.1476 0.0932>}]
 dyn_tatter_top_03 [{shapeType=Box, extentX=4, extentY=3.5, extentZ=1, center=<2.3605 -0.0001 -0.0001>}]
 dyn_tatters_master [{shapeType=Capsule, length=40, radius=4.25, center=<-4.1619 -0.0025 0.0916>, rotation=<rot 83.1814 -179.9999 -179.9999>}]
 lowerarm_r [{shapeType=Capsule, length=20, radius=4.0074, center=<-11.2288 0 0.0004>, rotation=<rot 90 0 0>}]
}
constraints [
 {bone1Name=dyn_tatter_top_02, bone2Name=dyn_tatter_top_01, swing1LimitDegrees=20, swing2LimitDegrees=30, twistLimitDegrees=3}
 {bone1Name=dyn_tatter_top_03, bone2Name=dyn_tatter_top_02, swing1LimitDegrees=20, swing2LimitDegrees=20, twistLimitDegrees=3}
 {bone1Name=dyn_tatter_top_01, bone2Name=dyn_tatters_master, swing1LimitDegrees=10, swing2LimitDegrees=20, twistLimitDegrees=3}
 {bone1Name=dyn_tatter_front, bone2Name=dyn_tatters_master, swing1LimitDegrees=1, swing2LimitDegrees=20, twistLimitDegrees=3}
 {bone1Name=dyn_tatter_bottom, bone2Name=dyn_tatters_master, swing1LimitDegrees=20, swing2LimitDegrees=20, twistLimitDegrees=3}
 {
  bone1Name=dyn_tatter_back_01
  bone2Name=dyn_tatters_master
  swing1LimitDegrees=2
  swing1Motion=Locked
  swing2LimitDegrees=10
  twistLimitDegrees=45
  twistMotion=Locked
 }
 {bone1Name=dyn_tatter_back_02, bone2Name=dyn_tatter_back_01, swing1LimitDegrees=10, swing2LimitDegrees=20, twistLimitDegrees=3}
 {bone1Name=dyn_tatter_back_03, bone2Name=dyn_tatter_back_02, swing1LimitDegrees=20, swing2LimitDegrees=20, twistLimitDegrees=3}
 {bone1Name=dyn_tatter_back_04, bone2Name=dyn_tatter_back_03, swing1LimitDegrees=20, swing2LimitDegrees=20, twistLimitDegrees=3}
]
