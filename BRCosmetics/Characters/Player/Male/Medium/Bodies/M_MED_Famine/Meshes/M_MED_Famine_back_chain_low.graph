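# Physics asset for the low back shoulder chain on the medium male Famine character body, named for dynamic chain bones that swing against the body.
# A capsule and a run of boxes form the chain from spine_04, ending in a larger box.
# Swing limits widen down the chain while twist stays locked or very tight.
# Capsules on pelvis, spine and right arm supply body shapes, linked by freely rotating constraints.
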PhysicsAsset /BRCosmetics/Characters/Player/Male/Medium/Bodies/M_MED_Famine/Meshes/M_MED_Famine_back_chain_low
bodies {
 dyn_shoulder_chain_back_02 [{shapeType=Capsule, length=0.5, radius=2}]
 dyn_shoulder_chain_back_03 [{shapeType=Box, extentX=4, extentY=4, extentZ=3.5}]
 dyn_shoulder_chain_back_04 [{shapeType=Box, extentX=3.5, extentY=3, extentZ=3.5, center=<0.6132 0 -0.0001>}]
 dyn_shoulder_chain_back_05 [{shapeType=Box, extentX=3, extentY=3.5, extentZ=3, center=<1.0881 0 0.0001>}]
 dyn_shoulder_chain_back_06 [{shapeType=Box, extentX=5, extentY=3, extentZ=3, center=<1.7551 0 -0.0002>}]
 dyn_shoulder_chain_back_07 [{shapeType=Box, extentX=3, extentY=1, extentZ=1, center=<1.5813 0 -0.0002>}]
 dyn_shoulder_chain_back_08 [{shapeType=Box, extentX=3.5, extentY=3.5, extentZ=3.5, center=<1.6323 0 -0.0001>}]
 dyn_shoulder_chain_back_09 [{shapeType=Box, extentX=10, extentY=6, extentZ=7.5, center=<6.2423 0 0>}]
 lowerarm_r [{shapeType=Capsule, length=14.0061, radius=5.7928, center=<-12.8703 0.0002 0.0001>, rotation=<rot 90 0 0>}]
 pelvis [{shapeType=Capsule, length=14.032, radius=9.65, rotation=<rot 0 90 0>}]
 spine_01 [
  {shapeType=Capsule, length=53.6318, radius=11.75, center=<18.0953 1.5068 6.0225>, rotation=<rot 0 90 90>}
  {shapeType=Capsule, length=27.8818, radius=11.75, center=<18.0953 -1.6265 -5.0655>, rotation=<rot 0 90 90>}
  {shapeType=Capsule, length=8.25, radius=13.5}
 ]
 spine_02 [{shapeType=Capsule, length=12.0556, radius=9.5452, center=<3.2313 -1.9456 0.0306>, rotation=<rot -3.9275 -90 0.407>}]
 spine_03 [{shapeType=Capsule, length=11.032, radius=10.2833, center=<4.0577 -1.6027 -0.5155>, rotation=<rot -1.5552 -90 0.5922>}]
 spine_04 [{shapeType=Capsule, length=11.032, radius=9.65, center=<4.0577 -1.6027 -0.5155>, rotation=<rot -1.5552 -90 0.5922>}]
 upperarm_r [{shapeType=Capsule, length=22.2561, radius=5.7928, center=<-7.7479 0.0002 0>, rotation=<rot 90 0 0>}]
}
constraints [
 {
  bone1Name=dyn_shoulder_chain_back_03
  bone2Name=dyn_shoulder_chain_back_02
  swing1LimitDegrees=10
  swing2LimitDegrees=10
  twistLimitDegrees=1
  twistMotion=Locked
 }
 {
  bone1Name=dyn_shoulder_chain_back_04
  bone2Name=dyn_shoulder_chain_back_03
  swing1LimitDegrees=20
  swing2LimitDegrees=20
  twistLimitDegrees=1
 }
 {
  bone1Name=dyn_shoulder_chain_back_05
  bone2Name=dyn_shoulder_chain_back_04
  swing1LimitDegrees=20
  swing2LimitDegrees=20
  twistLimitDegrees=1
 }
 {
  bone1Name=dyn_shoulder_chain_back_06
  bone2Name=dyn_shoulder_chain_back_05
  swing1LimitDegrees=20
  swing2LimitDegrees=20
  twistLimitDegrees=1
 }
 {
  bone1Name=dyn_shoulder_chain_back_07
  bone2Name=dyn_shoulder_chain_back_06
  swing1LimitDegrees=29
  swing2LimitDegrees=30
  twistLimitDegrees=1
 }
 {
  bone1Name=dyn_shoulder_chain_back_08
  bone2Name=dyn_shoulder_chain_back_07
  swing1LimitDegrees=29
  swing2LimitDegrees=30
  twistLimitDegrees=1
 }
 {
  bone1Name=dyn_shoulder_chain_back_09
  bone2Name=dyn_shoulder_chain_back_08
  swing1LimitDegrees=29
  swing2LimitDegrees=30
  twistLimitDegrees=1
 }
 {
  bone1Name=upperarm_r
  bone2Name=spine_04
  swing1LimitDegrees=45
  swing1Motion=Free
  swing2LimitDegrees=45
  swing2Motion=Free
  twistLimitDegrees=45
  twistMotion=Free
 }
 {
  bone1Name=dyn_shoulder_chain_back_02
  bone2Name=spine_04
  swing1LimitDegrees=45
  swing1Motion=Free
  swing2LimitDegrees=45
  swing2Motion=Free
  twistLimitDegrees=45
  twistMotion=Free
 }
 {
  bone1Name=spine_02
  bone2Name=spine_01
  swing1LimitDegrees=45
  swing1Motion=Free
  swing2LimitDegrees=45
  swing2Motion=Free
  twistLimitDegrees=45
  twistMotion=Free
 }
 {
  bone1Name=spine_03
  bone2Name=spine_02
  swing1LimitDegrees=45
  swing1Motion=Free
  swing2LimitDegrees=45
  swing2Motion=Free
  twistLimitDegrees=45
  twistMotion=Free
 }
 {
  bone1Name=spine_04
  bone2Name=spine_03
  swing1LimitDegrees=45
  swing1Motion=Free
  swing2LimitDegrees=45
  swing2Motion=Free
  twistLimitDegrees=45
  twistMotion=Free
 }
 {
  bone1Name=spine_01
  bone2Name=pelvis
  swing1LimitDegrees=45
  swing1Motion=Free
  swing2LimitDegrees=45
  swing2Motion=Free
  twistLimitDegrees=45
  twistMotion=Free
 }
 {
  bone1Name=lowerarm_r
  bone2Name=upperarm_r
  swing1LimitDegrees=45
  swing1Motion=Free
  swing2LimitDegrees=45
  swing2Motion=Free
  twistLimitDegrees=45
  twistMotion=Free
 }
]
