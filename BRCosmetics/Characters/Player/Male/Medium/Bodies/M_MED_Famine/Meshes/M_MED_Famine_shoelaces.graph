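# Physics asset for the shoelaces on the male medium Famine character body, giving the lace bones physical bodies anchored to the feet.
# A capsule covers each foot, and smaller capsules cover two shoelace bones per side.
# Each lace bone constrains directly to its foot rather than chaining to another lace, with twist locked and asymmetric swing limit angles.
# The right side uses different lace bones from the left (02 and 04 versus 02 and 05).
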PhysicsAsset /BRCosmetics/Characters/Player/Male/Medium/Bodies/M_MED_Famine/Meshes/M_MED_Famine_shoelaces
bodies {
 foot_l [{shapeType=Capsule, length=10.3195, radius=5.3, center=<-5.9967 -0.0001 0>, rotation=<rot 0 0 90>}]
 foot_r [{shapeType=Capsule, length=10.2033, radius=5, center=<5.0018 -1.9873 0.3652>, rotation=<rot 0 -176.8484 -88.5283>}]
 shoelace_l_02 [{shapeType=Capsule, length=3, radius=2, center=<2.8685 0 2.8179>, rotation=<rot -45.5096 -0.0001 0>}]
 shoelace_l_05 [{shapeType=Capsule, length=3, radius=2, center=<3.3749 -0.0001 0.1163>, rotation=<rot -79.385 -0.0004 0.0004>}]
 shoelace_r_02 [{shapeType=Capsule, length=3, radius=2, center=<2.9348 0.4477 -0.9021>, rotation=<rot 69.5481 -0.0001 -0.0001>}]
 shoelace_r_04 [{shapeType=Capsule, length=3, radius=2, center=<4.3428 0 -1.6207>, rotation=<rot -69.5342 -179.9999 179.9999>}]
}
constraints [
 {bone1Name=shoelace_l_02, bone2Name=foot_l, swing1LimitDegrees=10, swing2LimitDegrees=40, twistLimitDegrees=45, twistMotion=Locked}
 {bone1Name=shoelace_r_04, bone2Name=foot_r, swing1LimitDegrees=40, swing2LimitDegrees=5, twistLimitDegrees=45, twistMotion=Locked}
 {bone1Name=shoelace_r_02, bone2Name=foot_r, swing1LimitDegrees=10, swing2LimitDegrees=40, twistLimitDegrees=45, twistMotion=Locked}
 {bone1Name=shoelace_l_05, bone2Name=foot_l, swing1LimitDegrees=10, swing2LimitDegrees=40, twistLimitDegrees=45, twistMotion=Locked}
]
