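# Physics asset for the dangling zippers on the Famine medium male player body cosmetic, covering front, back and left cuff.
# Each zipper pairs a larger body on an odd-numbered bone with a small body on the following bone, joined by one constraint.
# Shapes are mostly boxes, with a few capsules.
# Limits are hinge-like: one wide swing axis, a narrower second swing, and minimal twist, locked on the back-left pairs.
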PhysicsAsset /BRCosmetics/Characters/Player/Male/Medium/Bodies/M_MED_Famine/Meshes/M_MED_Famine_zippers
bodies {
 dyn_zipper_back_l_01 [{shapeType=Box, extentX=19.01, extentY=20.01, extentZ=4.01, center=<0.0847 0 3.8923>, rotation=<rot 0 -90 0>}]
 dyn_zipper_back_l_02 [{shapeType=Box, extentX=1.7517, extentY=2.7049, extentZ=1.9916, center=<3.7314 0.0056 0.0162>, rotation=<rot 0 90.7416 -76.8764>}]
 dyn_zipper_back_l_03 [{shapeType=Box, extentX=18.5, extentY=19, extentZ=5.5, center=<1.5102 0.0001 4.8006>, rotation=<rot 0 -90 0>}]
 dyn_zipper_back_l_04 [{shapeType=Box, extentX=1.7517, extentY=2.7049, extentZ=1.9916, center=<3.7314 0.0056 0.0162>, rotation=<rot 0 90.7416 -76.8764>}]
 dyn_zipper_back_r_01 [{shapeType=Box, extentX=14, extentY=22.5, extentZ=1.5, center=<2.5999 0 2.4314>, rotation=<rot 0.0001 -89.9996 -12.7745>}]
 dyn_zipper_back_r_02 [{shapeType=Box, extentX=1.7517, extentY=4.7049, extentZ=1.9916, center=<3.7314 0.0056 0.0162>, rotation=<rot 0 90.7415 -70.5713>}]
 dyn_zipper_back_r_03 [{shapeType=Box, extentX=10.51, extentY=17.01, extentZ=5.01, center=<-0.8789 -0.0001 2.4378>, rotation=<rot 0 -89.9992 -14.9561>}]
 dyn_zipper_back_r_04 [{shapeType=Capsule, length=1.2572, radius=0.8655, center=<3.4498 -0.0069 0.2005>, rotation=<rot 0 89.3897 -83.9806>}]
 dyn_zipper_cuff_l_01 [{shapeType=Capsule, length=3, radius=3, center=<-7.992 4.5021 -9.9271>, rotation=<rot -17.7767 -100.8168 32.0367>}]
 dyn_zipper_cuff_l_02 [{shapeType=Capsule, length=2.2477, radius=0.8983, center=<3.1025 0.141 -0.1597>, rotation=<rot 0 -88.5413 -88.8254>}]
 dyn_zipper_front_l_01 [{shapeType=Box, extentX=11.2517, extentY=4.2049, extentZ=14.4916, center=<-1.202 -0.0581 1.8132>, rotation=<rot 0 90.7415 -76.872>}]
 dyn_zipper_front_l_02 [{shapeType=Box, extentX=1.7517, extentY=1.2049, extentZ=1.9916, center=<3.7314 0.0056 0.0162>, rotation=<rot 0 90.7416 -76.8764>}]
 dyn_zipper_front_l_03 [{shapeType=Box, extentX=9.7517, extentY=4.2049, extentZ=19.4916, center=<0.3359 -0.0384 2.5732>, rotation=<rot 0 90.7415 -76.8718>}]
 dyn_zipper_front_l_04 [{shapeType=Box, extentX=1.7517, extentY=1.2049, extentZ=1.9916, center=<3.7314 0.0056 0.0162>, rotation=<rot 0 90.7416 -76.8764>}]
 dyn_zipper_front_l_05 [{shapeType=Box, extentX=10.7517, extentY=6.7049, extentZ=20.9916, center=<1.9107 -0.0179 4.702>, rotation=<rot 0 90.7415 -76.8739>}]
 dyn_zipper_front_l_06 [{shapeType=Box, extentX=1.7517, extentY=1.2049, extentZ=1.9916, center=<3.7314 0.0056 0.0162>, rotation=<rot 0 90.7416 -76.8764>}]
 dyn_zipper_front_l_07 [{shapeType=Box, extentX=24.2517, extentY=13.7049, extentZ=4.9916, center=<4 -0.2945 3.5617>}]
 dyn_zipper_front_l_08 [{shapeType=Box, extentX=1.7517, extentY=2.2049, extentZ=3.4916, center=<4 -0.2946 0>}]
 dyn_zipper_front_r_01 [{shapeType=Box, extentX=10.01, extentY=4.51, extentZ=17.51, center=<3.1401 -0.0021 2.5527>, rotation=<rot 0 90.7415 -76.8742>}]
 dyn_zipper_front_r_02 [{shapeType=Box, extentX=1.7517, extentY=1.2049, extentZ=1.9916, center=<3.7314 0.0056 0.0162>, rotation=<rot 0 90.7416 -76.8764>}]
 dyn_zipper_front_r_03 [{shapeType=Box, extentX=10.01, extentY=1.01, extentZ=13.01, center=<1.5737 -0.0223 0.622>, rotation=<rot 0 90.7415 -76.873>}]
 dyn_zipper_front_r_04 [{shapeType=Box, extentX=1.7517, extentY=1.2049, extentZ=1.9916, center=<3.7314 0.0056 0.0162>, rotation=<rot 0 90.7416 -76.8764>}]
 dyn_zipper_front_r_05 [{shapeType=Box, extentX=10.2517, extentY=5.2049, extentZ=16.4916, center=<-0.9377 -0.0549 2.3609>, rotation=<rot 0 90.7415 -76.8723>}]
 dyn_zipper_front_r_06 [{shapeType=Box, extentX=1.7517, extentY=2.2049, extentZ=1.9916, center=<4 0 0>, rotation=<rot 0 90.7416 -76.8764>}]
 dyn_zipper_front_r_07 [{shapeType=Box, extentX=9.2517, extentY=9.7049, extentZ=12.4916, center=<0.0468 -0.042 5.1097>, rotation=<rot 0 90.7415 -76.8721>}]
 dyn_zipper_front_r_08 [{shapeType=Box, extentX=1.7517, extentY=1.7049, extentZ=1.9916, center=<3.6782 0.0049 0.2444>, rotation=<rot 0 90.7415 -76.876>}]
 dyn_zipper_front_r_09 [{shapeType=Box, extentX=16.51, extentY=17.51, extentZ=1.51, center=<3.1125 0.0002 2.1519>, rotation=<rot 0 -89.9995 -8.9087>}]
 dyn_zipper_front_r_10 [{shapeType=Box, extentX=1.7517, extentY=1.7049, extentZ=1.9916, center=<3.6782 0.0049 0.2444>, rotation=<rot 0 90.7415 -76.876>}]
}
constraints [
 {bone1Name=dyn_zipper_front_r_02, bone2Name=dyn_zipper_front_r_01, swing1LimitDegrees=20, swing2LimitDegrees=120, twistLimitDegrees=2}
 {bone1Name=dyn_zipper_front_r_04, bone2Name=dyn_zipper_front_r_03, swing1LimitDegrees=20, swing2LimitDegrees=120, twistLimitDegrees=2}
 {bone1Name=dyn_zipper_front_r_06, bone2Name=dyn_zipper_front_r_05, swing1LimitDegrees=20, swing2LimitDegrees=120, twistLimitDegrees=2}
 {bone1Name=dyn_zipper_front_r_08, bone2Name=dyn_zipper_front_r_07, swing1LimitDegrees=20, swing2LimitDegrees=120, twistLimitDegrees=2}
 {bone1Name=dyn_zipper_front_l_02, bone2Name=dyn_zipper_front_l_01, swing1LimitDegrees=20, swing2LimitDegrees=120, twistLimitDegrees=2}
 {bone1Name=dyn_zipper_front_l_04, bone2Name=dyn_zipper_front_l_03, swing1LimitDegrees=20, swing2LimitDegrees=120, twistLimitDegrees=2}
 {bone1Name=dyn_zipper_front_l_08, bone2Name=dyn_zipper_front_l_07, swing1LimitDegrees=20, swing2LimitDegrees=120, twistLimitDegrees=2}
 {bone1Name=dyn_zipper_front_l_06, bone2Name=dyn_zipper_front_l_05, swing1LimitDegrees=20, swing2LimitDegrees=120, twistLimitDegrees=2}
 {bone1Name=dyn_zipper_back_r_02, bone2Name=dyn_zipper_back_r_01, swing1LimitDegrees=8, swing2LimitDegrees=100, twistLimitDegrees=2}
 {
  bone1Name=dyn_zipper_back_l_04
  bone2Name=dyn_zipper_back_l_03
  swing1LimitDegrees=24.4286
  swing2LimitDegrees=100
  twistLimitDegrees=45
  twistMotion=Locked
 }
 {
  bone1Name=dyn_zipper_back_l_02
  bone2Name=dyn_zipper_back_l_01
  swing1LimitDegrees=20
  swing2LimitDegrees=100
  twistLimitDegrees=45
  twistMotion=Locked
 }
 {bone1Name=dyn_zipper_back_r_04, bone2Name=dyn_zipper_back_r_03, swing1LimitDegrees=5, swing2LimitDegrees=100, twistLimitDegrees=1}
 {bone1Name=dyn_zipper_cuff_l_02, bone2Name=dyn_zipper_cuff_l_01, swing1LimitDegrees=45, swing2LimitDegrees=90, twistLimitDegrees=1}
 {bone1Name=dyn_zipper_front_r_10, bone2Name=dyn_zipper_front_r_09, swing1LimitDegrees=20, swing2LimitDegrees=120, twistLimitDegrees=2}
]
