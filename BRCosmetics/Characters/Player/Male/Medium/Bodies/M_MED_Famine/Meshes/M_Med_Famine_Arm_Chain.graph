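# Physics asset for the arm chain on the medium male Famine player body cosmetic, covering its top, bottom and front dynamic chain bones.
# Each chain has two capsule bodies, a tiny one on the 01 bone and a larger one on the 02 bone.
# A constraint joins each pair, with no links between chains.
# All swing and twist motions are locked, so the stored angle limits leave no angular freedom.
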PhysicsAsset /BRCosmetics/Characters/Player/Male/Medium/Bodies/M_MED_Famine/Meshes/M_Med_Famine_Arm_Chain
bodies {
 dyn_arm_chain_bottom_01 [{shapeType=Capsule, length=0.505, radius=0.505, rotation=<rot 0 -90 0>}]
 dyn_arm_chain_bottom_02 [{shapeType=Capsule, length=6.7639, radius=3.9239, center=<0.3083 -0.8334 -0.1464>, rotation=<rot 0 -7.131 -81.6518>}]
 dyn_arm_chain_front_01 [{shapeType=Capsule, length=0.505, radius=0.505, rotation=<rot 0 -90 0>}]
 dyn_arm_chain_front_02 [{shapeType=Capsule, length=3, radius=3}]
 dyn_arm_chain_top_01 [{shapeType=Capsule, length=0.505, radius=0.505, rotation=<rot 0 -90 0>}]
 dyn_arm_chain_top_02 [{shapeType=Capsule, length=3, radius=3}]
}
constraints [
 {
  bone1Name=dyn_arm_chain_top_02
  bone2Name=dyn_arm_chain_top_01
  swing1LimitDegrees=3
  swing1Motion=Locked
  swing2LimitDegrees=3
  swing2Motion=Locked
  twistLimitDegrees=45
  twistMotion=Locked
 }
 {
  bone1Name=dyn_arm_chain_bottom_02
  bone2Name=dyn_arm_chain_bottom_01
  swing1LimitDegrees=45
  swing1Motion=Locked
  swing2LimitDegrees=45
  swing2Motion=Locked
  twistLimitDegrees=45
  twistMotion=Locked
 }
 {
  bone1Name=dyn_arm_chain_front_02
  bone2Name=dyn_arm_chain_front_01
  swing1LimitDegrees=3
  swing1Motion=Locked
  swing2LimitDegrees=3
  swing2Motion=Locked
  twistLimitDegrees=45
  twistMotion=Locked
 }
]
