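# Physics asset for the armpit chain on the male medium Famine character body, giving its dangling shoulder chains simulated secondary motion.
# Back and bottom chains each use two capsule bodies, a small one on the first bone and a larger one on the second.
# Each second bone is constrained to its first with tight angular limits: a few degrees for the bottom chain, barely any swing and locked twist for the back.
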PhysicsAsset /BRCosmetics/Characters/Player/Male/Medium/Bodies/M_MED_Famine/Meshes/M_Med_Famine_Armpit_Chain
bodies {
 dyn_shoulder_chain_back_01 [{shapeType=Capsule, length=0.505, radius=0.505, rotation=<rot 0 -90 0>}]
 dyn_shoulder_chain_back_02 [{shapeType=Capsule, length=2, radius=2, center=<0 -0.0002 -3.7514>}]
 dyn_shoulder_chain_bottom_01 [{shapeType=Capsule, length=0.505, radius=0.505, rotation=<rot 0 -90 0>}]
 dyn_shoulder_chain_bottom_02 [{shapeType=Capsule, length=2, radius=2}]
}
constraints [
 {
  bone1Name=dyn_shoulder_chain_bottom_02
  bone2Name=dyn_shoulder_chain_bottom_01
  swing1LimitDegrees=5
  swing2LimitDegrees=5
  twistLimitDegrees=5
 }
 {
  bone1Name=dyn_shoulder_chain_back_02
  bone2Name=dyn_shoulder_chain_back_01
  swing1LimitDegrees=1
  swing2LimitDegrees=1
  twistLimitDegrees=0
  twistMotion=Locked
 }
]
