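# Post-copy AnimBP for the Famine character body, adding secondary motion to chains, jacket, zippers and tatters over the attached parent mesh's pose.
# Targeting selects between two chain-physics branches, feeding shared RigidBody, bone-constraint and AnimDynamics passes.
# ADS or default Control Rigs follow, then a crouch rig blends in from the cached pose.
# EventGraph disables ankle-cuff physics while skydiving and updates ADS alphas only otherwise; zipper and shoelace RigidBody nodes are unconnected.
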AnimBlueprint /BRCosmetics/Characters/Player/Male/Medium/Bodies/M_MED_Famine/Meshes/M_MED_Famine_AnimBP
target_skeleton=&/M_MED_Famine_Skeleton
graphs [
 AnimGraph {nodes [
   RigidBody(Alpha=0.7) {
    bOverrideWorldGravity=true
    overridePhysicsAsset=&/M_MED_Famine_zippers
    simulationSpace=BaseBoneSpace
    alphaBoolBlend {blendOption=Linear}
    baseBoneRef {boneName=root}
    componentAppliedLinearAccClamp=<5000 5000 5000>
    componentLinearAccScale=<0.45 0.45 0.45>
    componentLinearVelScale=<0.45 0.45 0.45>
    overrideWorldGravity {z=-980}
   }
   RigidBody(Alpha=0.6) {
    overridePhysicsAsset=&/M_MED_Famine_shoelaces
    simulationSpace=BaseBoneSpace
    alphaBoolBlend {blendOption=Linear}
    baseBoneRef {boneName=root}
   }
   n1 = Control Rig(Source=Use cached pose 'PreADSEval'()) {
    alphaBoolBlend {blendOption=Linear}
    controlRigAssetReference {blueprintRigClass=&/M_MED_Famine_Crouch_CtrlRig.M_MED_Famine_Crouch_CtrlRig_C}
   }
   Output Pose(Result=Blend Poses by bool(BlendPose_0=@n1, BlendPose_1=Use cached pose 'PreADSEval'(), BlendTime_0=0.2, bActiveValue=Get bIsCrouching()) {blendType=Linear}) {layerGroup=None}
   n2 = RigidBody(ComponentPose=Local To Component(LocalPose=Copy Pose From Mesh() {bUseAttachedParent=true})) {
    overridePhysicsAsset=&/M_Med_Famine_Arm_Chain
    simulationSpace=BaseBoneSpace
    alphaBoolBlend {blendOption=Linear}
    baseBoneRef {boneName=pelvis}
   }
   n3 = RigidBody(ComponentPose=@n2, Alpha=0.8) {
    bClampLinearTranslationLimitToRefPose=true
    bOverrideWorldGravity=true
    overridePhysicsAsset=&/M_MED_Famine_back_chain_low
    simulationSpace=BaseBoneSpace
    alphaBoolBlend {blendOption=Linear}
    baseBoneRef {boneName=pelvis}
    componentLinearAccScale {z=0.67}
    componentLinearVelScale {z=0.35}
    overrideWorldGravity {z=-1800}
   }
   n4 = Component To Local(ComponentPose=@n3)
   n5 = RigidBody(ComponentPose=Local To Component(LocalPose=Copy Pose From Mesh() {bUseAttachedParent=true}), Alpha=0.5) {
    overridePhysicsAsset=&/M_Med_Famine_Armpit_Chain
    simulationSpace=BaseBoneSpace
    alphaBoolBlend {blendOption=Linear}
    baseBoneRef {boneName=root}
   }
   n6 = "Transform (Modify) Bone - Bone: dyn_shoulder_chain_top_01"(ComponentPose=@n5, Alpha=1, Translation="2,0,0", Rotation="0,0,0", Scale="1,1,1") {
    translationMode=BMM_Additive
    translationSpace=BCS_BoneSpace
    alphaBoolBlend {blendOption=Linear}
    boneToModify {boneName=dyn_shoulder_chain_top_01}
    translation {x=2}
   }
   n7 = RigidBody(ComponentPose=@n6, Alpha=0.8) {
    bClampLinearTranslationLimitToRefPose=true
    bOverrideWorldGravity=true
    overridePhysicsAsset=&M_MED_Famine_back_chain_low
    simulationSpace=BaseBoneSpace
    alphaBoolBlend {blendOption=Linear}
    baseBoneRef {boneName=pelvis}
    componentLinearAccScale {z=0.67}
    componentLinearVelScale {z=0.35}
    overrideWorldGravity {z=-1900}
   }
   n8 = Component To Local(ComponentPose=@n7)
   n9 = Get bIsTargeting()
   n10 = "Constraint - dyn_zipper_cuff_l_01 (P)"(
    ComponentPose=Local To Component(LocalPose=Blend Poses by bool(BlendPose_0=@n8, BlendPose_1=@n4, bActiveValue=@n9) {blendType=Linear})
    Alpha=1
    ConstraintWeights_0=1
   ) {
    alphaBoolBlend {blendOption=Linear}
    boneToModify {boneName=dyn_zipper_cuff_l_01}
    constraintSetup [{targetBone {boneName=lowerarm_l}}]
    constraintWeights [1]
   }
   n11 = "Transform (Modify) Bone - Bone: dyn_zipper_cuff_l_01"(ComponentPose=@n10, Alpha=1, Translation="0.5,0,0", Rotation="0,0,0", Scale="1,1,1") {
    translationMode=BMM_Additive
    translationSpace=BCS_BoneSpace
    alphaBoolBlend {blendOption=Linear}
    boneToModify {boneName=dyn_zipper_cuff_l_01}
    translation {x=0.5}
   }
   n12 = RigidBody(ComponentPose=@n11) {
    bClampLinearTranslationLimitToRefPose=true
    bOverrideWorldGravity=true
    overridePhysicsAsset=&/M_MED_Famine_Jacket
    simulationSpace=BaseBoneSpace
    alphaBoolBlend {blendOption=Linear}
    baseBoneRef {boneName=pelvis}
    componentAppliedLinearAccClamp=<5000 5000 5000>
    componentLinearAccScale=<0.35 0.35 0.5>
    componentLinearVelScale=<0.35 0.35 0.5>
    overrideWorldGravity {z=-980}
   }
   n13 = RigidBody(ComponentPose=@n12, Alpha=0.7) {
    overridePhysicsAsset=&/M_MED_Famine_ChestChain
    simulationSpace=BaseBoneSpace
    alphaBoolBlend {blendOption=Linear}
    baseBoneRef {boneName=pelvis}
    componentLinearAccScale {z=0.25}
    componentLinearVelScale {z=0.25}
   }
   n14 = RigidBody(ComponentPose=@n13, Alpha=Get skyDivingAlpha() -> (skyDivingAlpha:double)) {
    overridePhysicsAsset=&/M_MED_Famine_AnkleCuff
    simulationSpace=BaseBoneSpace
    alphaBoolBlend {blendOption=Linear}
    baseBoneRef {boneName=calf_r}
    componentLinearAccScale=<1 1 1>
   }
   n15 = RigidBody(ComponentPose=@n14, Alpha=0.7) {
    bClampLinearTranslationLimitToRefPose=true
    bOverrideWorldGravity=true
    overridePhysicsAsset=&/M_MED_Famine_SleeveTatters
    simulationSpace=BaseBoneSpace
    alphaBoolBlend {blendOption=Linear}
    baseBoneRef {boneName=clavicle_r}
    componentAppliedLinearAccClamp=<6000 6000 6000>
    componentLinearAccScale=<0.15 0.15 0.15>
    componentLinearVelScale=<0.15 0.15 0.15>
    overrideWorldGravity {z=-980}
   }
   n16 = "Constraint - dyn_zipper_back_r_01 (P)"!(ComponentPose=@n15, Alpha=1, ConstraintWeights_0=1, ConstraintWeights_1:float=0) -> (Pose) {
    alphaBoolBlend {blendOption=Linear}
    boneToModify {boneName=dyn_zipper_back_r_01}
    constraintSetup [{targetBone {boneName=dyn_jacket_b_r_02}}]
    constraintWeights [1,0]
   }
   n17 = "Constraint - dyn_zipper_back_r_03 (PP)"!(ComponentPose=@n16, Alpha=1, ConstraintWeights_0=0.0395, ConstraintWeights_1:float=0.9605) -> (Pose) {
    alphaBoolBlend {blendOption=Linear}
    boneToModify {boneName=dyn_zipper_back_r_03}
    constraintSetup [{targetBone {boneName=dyn_jacket_r_02}}, {targetBone {boneName=dyn_jacket_b_r_02}}]
    constraintWeights [1,1]
   }
   n18 = "Constraint - dyn_zipper_back_l_01 (PPPPPPP)"!(
    ComponentPose=@n17
    Alpha=1
    ConstraintWeights_0=0.2985
    ConstraintWeights_1:float=0.2143
    ConstraintWeights_2:float=0.1742
    ConstraintWeights_3:float=0.0941
    ConstraintWeights_4:float=0.0533
    ConstraintWeights_5:float=0.156
    ConstraintWeights_6:float=0.0096
   ) -> (Pose) {
    alphaBoolBlend {blendOption=Linear}
    boneToModify {boneName=dyn_zipper_back_l_01}
    constraintSetup [
     {targetBone {boneName=dyn_jacket_l_02}}
     {targetBone {boneName=spine_01}}
     {targetBone {boneName=spine_02}}
     {targetBone {boneName=spine_03}}
     {targetBone {boneName=dyn_jacket_b_l_02}}
     {targetBone {boneName=pelvis}}
     {targetBone {boneName=spine_04}}
    ]
    constraintWeights [1,1,1,1,1,1,1]
   }
   n19 = "Constraint - dyn_zipper_back_l_03 (PPPPPPP)"!(
    ComponentPose=@n18
    Alpha=1
    ConstraintWeights_0=0.8199
    ConstraintWeights_1:float=0.0064
    ConstraintWeights_2:float=0.005
    ConstraintWeights_3:float=0.0028
    ConstraintWeights_4:float=0.1607
    ConstraintWeights_5:float=0.0048
    ConstraintWeights_6:float=0.0003
   ) -> (Pose) {
    alphaBoolBlend {blendOption=Linear}
    boneToModify {boneName=dyn_zipper_back_l_03}
    constraintSetup [
     {targetBone {boneName=dyn_jacket_l_02}}
     {targetBone {boneName=spine_01}}
     {targetBone {boneName=spine_02}}
     {targetBone {boneName=spine_03}}
     {targetBone {boneName=dyn_jacket_b_l_02}}
     {targetBone {boneName=pelvis}}
     {targetBone {boneName=spine_04}}
    ]
    constraintWeights [1,1,1,1,1,1,1]
   }
   n20 = "Constraint - dyn_zipper_front_r_09 (PPP)"!(ComponentPose=@n19, Alpha=1, ConstraintWeights_0=0.0006, ConstraintWeights_1:float=0.2494, ConstraintWeights_2:float=0.75) -> (Pose) {
    alphaBoolBlend {blendOption=Linear}
    boneToModify {boneName=dyn_zipper_front_r_09}
    constraintSetup [{targetBone {boneName=dyn_jacket_r_02}}, {targetBone {boneName=dyn_jacket_f_r_02}}, {targetBone {boneName=dyn_jacket_f_r_03}}]
    constraintWeights [1,1,1]
   }
   n21 = "Constraint - dyn_zipper_front_l_01 (PPPPP)"!(
    ComponentPose=@n20
    Alpha=1
    ConstraintWeights_0=0.0498
    ConstraintWeights_1:float=0.1995
    ConstraintWeights_2:float=0.5562
    ConstraintWeights_3:float=0.1886
    ConstraintWeights_4:float=0.0059
   ) -> (Pose) {
    alphaBoolBlend {blendOption=Linear}
    boneToModify {boneName=dyn_zipper_front_l_01}
    constraintSetup [
     {targetBone {boneName=spine_02}}
     {targetBone {boneName=spine_03}}
     {targetBone {boneName=spine_04}}
     {targetBone {boneName=spine_05}}
     {targetBone {boneName=clavicle_l}}
    ]
    constraintWeights [1,1,1,1,1]
   }
   n22 = "Constraint - dyn_zipper_front_l_03 (PPPP)"!(
    ComponentPose=@n21
    Alpha=1
    ConstraintWeights_0=0.1817
    ConstraintWeights_1:float=0.3582
    ConstraintWeights_2:float=0.4231
    ConstraintWeights_3:float=0.037
   ) -> (Pose) {
    alphaBoolBlend {blendOption=Linear}
    boneToModify {boneName=dyn_zipper_front_l_03}
    constraintSetup [{targetBone {boneName=spine_02}}, {targetBone {boneName=spine_03}}, {targetBone {boneName=spine_04}}, {targetBone {boneName=spine_05}}]
    constraintWeights [1,1,1,1]
   }
   n23 = "Constraint - dyn_zipper_front_r_07 (P)"!(ComponentPose=@n22, Alpha=1, ConstraintWeights_0=1, ConstraintWeights_1:float=0) -> (Pose) {
    alphaBoolBlend {blendOption=Linear}
    boneToModify {boneName=dyn_zipper_front_r_07}
    constraintSetup [{targetBone {boneName=dyn_jacket_f_r_02}}]
    constraintWeights [1,0]
   }
   n24 = "Constraint - dyn_zipper_front_r_03 (PPPPP)"!(
    ComponentPose=@n23
    Alpha=1
    ConstraintWeights_0=0.0002
    ConstraintWeights_1:float=0.1825
    ConstraintWeights_2:float=0.3595
    ConstraintWeights_3:float=0.4217
    ConstraintWeights_4:float=0.036
   ) -> (Pose) {
    alphaBoolBlend {blendOption=Linear}
    boneToModify {boneName=dyn_zipper_front_r_03}
    constraintSetup [
     {targetBone {boneName=spine_01}}
     {targetBone {boneName=spine_02}}
     {targetBone {boneName=spine_03}}
     {targetBone {boneName=spine_04}}
     {targetBone {boneName=spine_05}}
    ]
    constraintWeights [1,1,1,1,1]
   }
   n25 = "Constraint - dyn_zipper_cuff_l_01 (PPPPPP)"!(
    ComponentPose=@n24
    Alpha=1
    ConstraintWeights_0=0.0456
    ConstraintWeights_1:float=0.0117
    ConstraintWeights_2:float=0.0075
    ConstraintWeights_3:float=0.0555
    ConstraintWeights_4:float=0.6091
    ConstraintWeights_5:float=0.2706
   ) -> (Pose) {
    alphaBoolBlend {blendOption=Linear}
    boneToModify {boneName=dyn_zipper_cuff_l_01}
    constraintSetup [
     {targetBone {boneName=pinky_metacarpal_l}}
     {targetBone {boneName=hand_l}}
     {targetBone {boneName=thumb_01_l}}
     {targetBone {boneName=lowerarm_l}}
     {targetBone {boneName=lowerarm_twist_01_l}}
     {targetBone {boneName=lowerarm_twist_02_l}}
    ]
    constraintWeights [1,1,1,1,1,1]
   }
   n26 = "Constraint - dyn_zipper_front_l_07 (PPPPP)"!(
    ComponentPose=@n25
    Alpha=1
    ConstraintWeights_0=0.3491
    ConstraintWeights_1:float=0.1686
    ConstraintWeights_2:float=0.0743
    ConstraintWeights_3:float=0.0004
    ConstraintWeights_4:float=0.4077
   ) -> (Pose) {
    alphaBoolBlend {blendOption=Linear}
    boneToModify {boneName=dyn_zipper_front_l_07}
    constraintSetup [
     {targetBone {boneName=pelvis}}
     {targetBone {boneName=spine_02}}
     {targetBone {boneName=spine_03}}
     {targetBone {boneName=spine_04}}
     {targetBone {boneName=spine_01}}
    ]
    constraintWeights [1,1,1,1,1]
   }
   n27 = "Constraint - dyn_zipper_front_l_05 (PPPPP)"!(
    ComponentPose=@n26
    Alpha=1
    ConstraintWeights_0=0.2534
    ConstraintWeights_1:float=0.2385
    ConstraintWeights_2:float=0.1291
    ConstraintWeights_3:float=0.0166
    ConstraintWeights_4:float=0.3623
   ) -> (Pose) {
    alphaBoolBlend {blendOption=Linear}
    boneToModify {boneName=dyn_zipper_front_l_05}
    constraintSetup [
     {targetBone {boneName=pelvis}}
     {targetBone {boneName=spine_02}}
     {targetBone {boneName=spine_03}}
     {targetBone {boneName=spine_04}}
     {targetBone {boneName=spine_01}}
    ]
    constraintWeights [1,1,1,1,1]
   }
   n28 = "Constraint - dyn_zipper_front_r_05 (PPPPPPP)"!(
    ComponentPose=@n27
    Alpha=1
    ConstraintWeights_0=0.0011
    ConstraintWeights_1:float=0.224
    ConstraintWeights_2:float=0.3119
    ConstraintWeights_3:float=0.2586
    ConstraintWeights_4:float=0.0786
    ConstraintWeights_5:float=0.1183
    ConstraintWeights_6:float=0.0076
   ) -> (Pose) {
    alphaBoolBlend {blendOption=Linear}
    boneToModify {boneName=dyn_zipper_front_r_05}
    constraintSetup [
     {targetBone {boneName=dyn_jacket_r_02}}
     {targetBone {boneName=spine_01}}
     {targetBone {boneName=spine_02}}
     {targetBone {boneName=spine_03}}
     {targetBone {boneName=spine_04}}
     {targetBone {boneName=pelvis}}
     {targetBone {boneName=dyn_jacket_f_r_02}}
    ]
    constraintWeights [1,1,1,1,1,1,1]
   }
   n29 = "Constraint - dyn_zipper_front_r_01 (PPPP)"!(
    ComponentPose=@n28
    Alpha=1
    ConstraintWeights_0=0.0716
    ConstraintWeights_1:float=0.2315
    ConstraintWeights_2:float=0.5735
    ConstraintWeights_3:float=0.1234
   ) -> (Pose) {
    alphaBoolBlend {blendOption=Linear}
    boneToModify {boneName=dyn_zipper_front_r_01}
    constraintSetup [{targetBone {boneName=spine_02}}, {targetBone {boneName=spine_03}}, {targetBone {boneName=spine_04}}, {targetBone {boneName=spine_05}}]
    constraintWeights [1,1,1,1]
   }
   n30 = Anim Dynamics - Bone: dyn_zipper_back_l_03(ComponentPose=@n29, Alpha=1, ExternalForce="0,0,0") {
    angularBiasOverride=0.9999
    angularDampingOverride=1
    bGravityOverrideInSimSpace=true
    bOverrideAngularDamping=true
    bOverrideLinearDamping=true
    linearDampingOverride=1
    alphaBoolBlend {blendOption=Linear}
    boundBone {boneName=dyn_zipper_back_l_03}
    componentAppliedLinearAccClamp=<6000 6000 6000>
    componentLinearAccScale=<0.15 0.15 0.15>
    componentLinearVelScale=<0.15 0.15 0.15>
    physicsBodyDefinitions [{
      sphereCollisionRadius=0
      boundBone {boneName=dyn_zipper_back_l_03}
      boxExtents=<7 6 3>
      constraintSetup {angularLimitsMax=<0 30 30>, angularLimitsMin=<0 -5 -30>, angularTarget=<0 0 0>}
      localJointOffset=<3 0 0>
     }]
    relativeSpaceBone {boneName=root}
   }
   n31 = Anim Dynamics - Bone: dyn_zipper_back_l_01(ComponentPose=@n30, Alpha=1, ExternalForce="0,0,0") {
    angularBiasOverride=0.9999
    angularDampingOverride=0.9999
    bGravityOverrideInSimSpace=true
    bOverrideAngularDamping=true
    bOverrideLinearDamping=true
    linearDampingOverride=0.9999
    alphaBoolBlend {blendOption=Linear}
    boundBone {boneName=dyn_zipper_back_l_01}
    componentAppliedLinearAccClamp=<6000 6000 6000>
    componentLinearAccScale=<0.15 0.15 0.15>
    componentLinearVelScale=<0.15 0.15 0.15>
    physicsBodyDefinitions [{
      sphereCollisionRadius=0
      boundBone {boneName=dyn_zipper_back_l_01}
      boxExtents=<7 4 2>
      constraintSetup {angularLimitsMax=<0 30 30>, angularLimitsMin=<0 -5 -30>, angularTarget=<0 0 0>}
      localJointOffset=<2 0 0>
     }]
    relativeSpaceBone {boneName=pelvis}
   }
   n32 = Anim Dynamics - Bone: dyn_zipper_back_r_01(ComponentPose=@n31, Alpha=1, ExternalForce="0,0,0") {
    angularBiasOverride=0.9999
    angularDampingOverride=0.9999
    bGravityOverrideInSimSpace=true
    bOverrideAngularDamping=true
    bOverrideLinearDamping=true
    linearDampingOverride=0.9999
    alphaBoolBlend {blendOption=Linear}
    boundBone {boneName=dyn_zipper_back_r_01}
    componentAppliedLinearAccClamp=<6000 6000 6000>
    componentLinearAccScale=<0.15 0.15 0.15>
    componentLinearVelScale=<0.15 0.15 0.15>
    physicsBodyDefinitions [{
      sphereCollisionRadius=0
      boundBone {boneName=dyn_zipper_back_r_01}
      boxExtents=<10 8 3>
      constraintSetup {angularLimitsMax=<0 30 30>, angularLimitsMin=<0 -5 -30>, angularTarget=<0 0 0>}
      localJointOffset=<4 0 0>
     }]
    relativeSpaceBone {boneName=root}
   }
   n33 = Anim Dynamics - Bone: dyn_zipper_back_r_03(ComponentPose=@n32, Alpha=1, ExternalForce="0,0,0") {
    angularBiasOverride=0.9999
    angularDampingOverride=0.9999
    bGravityOverrideInSimSpace=true
    bOverrideAngularDamping=true
    bOverrideLinearDamping=true
    linearDampingOverride=0.9999
    alphaBoolBlend {blendOption=Linear}
    boundBone {boneName=dyn_zipper_back_r_03}
    componentAppliedLinearAccClamp=<6000 6000 6000>
    componentLinearAccScale=<0.15 0.15 0.15>
    componentLinearVelScale=<0.15 0.15 0.15>
    physicsBodyDefinitions [{
      sphereCollisionRadius=0
      boundBone {boneName=dyn_zipper_back_r_03}
      boxExtents=<7 4 2>
      constraintSetup {angularLimitsMax=<0 30 30>, angularLimitsMin=<0 -5 -30>, angularTarget=<0 0 0>}
      localJointOffset=<2 0 0>
     }]
    relativeSpaceBone {boneName=pelvis}
   }
   n34 = Anim Dynamics - Bone: dyn_zipper_front_r_08(ComponentPose=@n33, Alpha=1, ExternalForce="0,0,0") {
    angularBiasOverride=0.9999
    angularDampingOverride=0.9999
    bGravityOverrideInSimSpace=true
    bOverrideAngularDamping=true
    bOverrideLinearDamping=true
    linearDampingOverride=0.9999
    alphaBoolBlend {blendOption=Linear}
    boundBone {boneName=dyn_zipper_front_r_08}
    componentAppliedLinearAccClamp=<6000 6000 6000>
    componentLinearAccScale=<0.15 0.15 0.15>
    componentLinearVelScale=<0.15 0.15 0.15>
    physicsBodyDefinitions [{
      sphereCollisionRadius=0
      boundBone {boneName=dyn_zipper_front_r_08}
      boxExtents=<7 4 2>
      constraintSetup {angularLimitsMax=<0 30 30>, angularLimitsMin=<0 -5 -30>, angularTarget=<0 0 0>}
      localJointOffset=<2 0 0>
     }]
    relativeSpaceBone {boneName=pelvis}
   }
   n35 = Anim Dynamics - Bone: dyn_zipper_front_r_06(ComponentPose=@n34, Alpha=1, ExternalForce="0,0,0") {
    angularBiasOverride=0.9999
    angularDampingOverride=0.9999
    bGravityOverrideInSimSpace=true
    bOverrideAngularDamping=true
    bOverrideLinearDamping=true
    linearDampingOverride=0.9999
    alphaBoolBlend {blendOption=Linear}
    boundBone {boneName=dyn_zipper_front_r_06}
    componentAppliedLinearAccClamp=<6000 6000 6000>
    componentLinearAccScale=<0.15 0.15 0.15>
    componentLinearVelScale=<0.15 0.15 0.15>
    physicsBodyDefinitions [{
      sphereCollisionRadius=0
      boundBone {boneName=dyn_zipper_front_r_06}
      boxExtents=<10 8 3>
      constraintSetup {angularLimitsMax=<0 30 30>, angularLimitsMin=<0 -5 -30>, angularTarget=<0 0 0>}
      localJointOffset=<4 0 0>
     }]
    relativeSpaceBone {boneName=root}
   }
   n36 = Anim Dynamics - Bone: dyn_zipper_front_r_04(ComponentPose=@n35, Alpha=1, ExternalForce="0,0,0") {
    angularBiasOverride=0.9999
    angularDampingOverride=0.9999
    bGravityOverrideInSimSpace=true
    bOverrideAngularDamping=true
    bOverrideLinearDamping=true
    linearDampingOverride=0.9999
    alphaBoolBlend {blendOption=Linear}
    boundBone {boneName=dyn_zipper_front_r_04}
    componentAppliedLinearAccClamp=<6000 6000 6000>
    componentLinearAccScale=<0.15 0.15 0.15>
    componentLinearVelScale=<0.15 0.15 0.15>
    physicsBodyDefinitions [{
      sphereCollisionRadius=0
      boundBone {boneName=dyn_zipper_front_r_04}
      boxExtents=<7 4 2>
      constraintSetup {angularLimitsMax=<0 30 30>, angularLimitsMin=<0 -5 -30>, angularTarget=<0 0 0>}
      localJointOffset=<2 0 0>
     }]
    relativeSpaceBone {boneName=pelvis}
   }
   n37 = Anim Dynamics - Bone: dyn_zipper_front_r_02(ComponentPose=@n36, Alpha=1, ExternalForce="0,0,0") {
    angularBiasOverride=0.9999
    angularDampingOverride=0.9999
    bGravityOverrideInSimSpace=true
    bOverrideAngularDamping=true
    bOverrideLinearDamping=true
    linearDampingOverride=0.9999
    alphaBoolBlend {blendOption=Linear}
    boundBone {boneName=dyn_zipper_front_r_02}
    componentAppliedLinearAccClamp=<6000 6000 6000>
    componentLinearAccScale=<0.15 0.15 0.15>
    componentLinearVelScale=<0.15 0.15 0.15>
    physicsBodyDefinitions [{
      sphereCollisionRadius=0
      boundBone {boneName=dyn_zipper_front_r_02}
      boxExtents=<7 6 3>
      constraintSetup {angularLimitsMax=<0 30 30>, angularLimitsMin=<0 -5 -30>, angularTarget=<0 0 0>}
      localJointOffset=<3 0 0>
     }]
    relativeSpaceBone {boneName=root}
   }
   n38 = Anim Dynamics - Bone: dyn_zipper_front_l_02(ComponentPose=@n37, Alpha=1, ExternalForce="0,0,0") {
    angularBiasOverride=0.9999
    angularDampingOverride=0.9999
    bGravityOverrideInSimSpace=true
    bOverrideAngularDamping=true
    bOverrideLinearDamping=true
    linearDampingOverride=0.9999
    alphaBoolBlend {blendOption=Linear}
    boundBone {boneName=dyn_zipper_front_l_02}
    componentAppliedLinearAccClamp=<6000 6000 6000>
    componentLinearAccScale=<0.15 0.15 0.15>
    componentLinearVelScale=<0.15 0.15 0.15>
    physicsBodyDefinitions [{
      sphereCollisionRadius=0
      boundBone {boneName=dyn_zipper_front_l_02}
      boxExtents=<7 6 3>
      constraintSetup {angularLimitsMax=<0 30 30>, angularLimitsMin=<0 -5 -30>, angularTarget=<0 0 0>}
      localJointOffset=<3 0 0>
     }]
    relativeSpaceBone {boneName=root}
   }
   n39 = Anim Dynamics - Bone: dyn_zipper_front_l_04(ComponentPose=@n38, Alpha=1, ExternalForce="0,0,0") {
    angularBiasOverride=0.9999
    angularDampingOverride=0.9999
    bGravityOverrideInSimSpace=true
    bOverrideAngularDamping=true
    bOverrideLinearDamping=true
    linearDampingOverride=0.9999
    alphaBoolBlend {blendOption=Linear}
    boundBone {boneName=dyn_zipper_front_l_04}
    componentAppliedLinearAccClamp=<6000 6000 6000>
    componentLinearAccScale=<0.15 0.15 0.15>
    componentLinearVelScale=<0.15 0.15 0.15>
    physicsBodyDefinitions [{
      sphereCollisionRadius=0
      boundBone {boneName=dyn_zipper_front_l_04}
      boxExtents=<7 4 2>
      constraintSetup {angularLimitsMax=<0 30 30>, angularLimitsMin=<0 -5 -30>, angularTarget=<0 0 0>}
      localJointOffset=<2 0 0>
     }]
    relativeSpaceBone {boneName=pelvis}
   }
   n40 = Anim Dynamics - Bone: dyn_zipper_front_l_06(ComponentPose=@n39, Alpha=1, ExternalForce="0,0,0") {
    angularBiasOverride=0.9999
    angularDampingOverride=0.9999
    bGravityOverrideInSimSpace=true
    bOverrideAngularDamping=true
    bOverrideLinearDamping=true
    linearDampingOverride=0.9999
    alphaBoolBlend {blendOption=Linear}
    boundBone {boneName=dyn_zipper_front_l_06}
    componentAppliedLinearAccClamp=<6000 6000 6000>
    componentLinearAccScale=<0.15 0.15 0.15>
    componentLinearVelScale=<0.15 0.15 0.15>
    physicsBodyDefinitions [{
      sphereCollisionRadius=0
      boundBone {boneName=dyn_zipper_front_l_06}
      boxExtents=<10 8 3>
      constraintSetup {angularLimitsMax=<0 30 30>, angularLimitsMin=<0 -5 -30>, angularTarget=<0 0 0>}
      localJointOffset=<4 0 0>
     }]
    relativeSpaceBone {boneName=root}
   }
   n41 = Anim Dynamics - Bone: dyn_zipper_front_l_08(ComponentPose=@n40, Alpha=1, ExternalForce="0,0,0") {
    angularBiasOverride=0.9999
    angularDampingOverride=0.9999
    bGravityOverrideInSimSpace=true
    bOverrideAngularDamping=true
    bOverrideLinearDamping=true
    linearDampingOverride=0.9999
    alphaBoolBlend {blendOption=Linear}
    boundBone {boneName=dyn_zipper_front_l_08}
    componentAppliedLinearAccClamp=<6000 6000 6000>
    componentLinearAccScale=<0.15 0.15 0.15>
    componentLinearVelScale=<0.15 0.15 0.15>
    physicsBodyDefinitions [{
      sphereCollisionRadius=0
      boundBone {boneName=dyn_zipper_front_l_08}
      boxExtents=<7 4 2>
      constraintSetup {angularLimitsMax=<0 30 30>, angularLimitsMin=<0 -5 -30>, angularTarget=<0 0 0>}
      localJointOffset=<2 0 0>
     }]
    relativeSpaceBone {boneName=pelvis}
   }
   n42 = Anim Dynamics - Bone: dyn_zipper_cuff_l_02(ComponentPose=@n41, Alpha=1, ExternalForce="0,0,0") {
    angularBiasOverride=0.9999
    angularDampingOverride=0.9999
    bGravityOverrideInSimSpace=true
    bOverrideAngularDamping=true
    bOverrideLinearDamping=true
    linearDampingOverride=0.9999
    alphaBoolBlend {blendOption=Linear}
    boundBone {boneName=dyn_zipper_cuff_l_02}
    componentAppliedLinearAccClamp=<6000 6000 6000>
    componentLinearAccScale=<0.15 0.15 0.15>
    componentLinearVelScale=<0.15 0.15 0.15>
    physicsBodyDefinitions [{
      sphereCollisionRadius=0
      boundBone {boneName=dyn_zipper_cuff_l_02}
      boxExtents=<7 4 2>
      constraintSetup {angularLimitsMax=<0 30 30>, angularLimitsMin=<0 -5 -30>, angularTarget=<0 0 0>}
      localJointOffset=<2 0 0>
     }]
    relativeSpaceBone {boneName=pelvis}
   }
   n43 = Anim Dynamics - Bone: dyn_zipper_front_r_10(ComponentPose=@n42, Alpha=1, ExternalForce="0,0,0") {
    angularBiasOverride=0.9999
    angularDampingOverride=0.9999
    bGravityOverrideInSimSpace=true
    bOverrideAngularDamping=true
    bOverrideLinearDamping=true
    linearDampingOverride=0.9999
    alphaBoolBlend {blendOption=Linear}
    boundBone {boneName=dyn_zipper_front_r_10}
    componentAppliedLinearAccClamp=<6000 6000 6000>
    componentLinearAccScale=<0.15 0.15 0.15>
    componentLinearVelScale=<0.15 0.15 0.15>
    physicsBodyDefinitions [{
      sphereCollisionRadius=0
      boundBone {boneName=dyn_zipper_front_r_10}
      boxExtents=<7 6 3>
      constraintSetup {angularLimitsMax=<0 30 30>, angularLimitsMin=<0 -5 -30>, angularTarget=<0 0 0>}
      localJointOffset=<3 0 0>
     }]
    relativeSpaceBone {boneName=root}
   }
   n44 = Control Rig(Source=Component To Local(ComponentPose=@n43), Alpha=Get ADS_Alpha() -> (ADS_Alpha:double)) {
    alphaBoolBlend {blendOption=Linear}
    controlRigAssetReference {blueprintRigClass=&/M_MED_Famine_ADS_CtrlRig.M_MED_Famine_ADS_CtrlRig_C}
   }
   n45 = Control Rig(Source=@n44, Alpha=Get Non_ADS_Alpha() -> (Non_ADS_Alpha:double)) {
    alphaBoolBlend {blendOption=Linear}
    controlRigAssetReference {blueprintRigClass=&/M_MED_Famine_CtrlRig.M_MED_Famine_CtrlRig_C}
   }
   Save cached pose 'PreADSEval'(Pose=@n45) {cacheName=PreADSEval}
   Reroute Node(InputPin:Wildcard) -> (OutputPin:Wildcard)
  ]}
 EventGraph {nodes [
   Event BlueprintUpdateAnimation()
   > n1 = Branch(Condition=Get bIsSkyDiving())
   > Set skyDivingAlpha(skyDivingAlpha:double=0) -> (Output_Get:double)
   Set skyDivingAlpha(execute=@n1.else, skyDivingAlpha=1)
   > n2 = Branch(Condition=Get bIsTargeting())
   > Set ADS_Alpha(ADS_Alpha:double=1) -> (Output_Get:double)
   > Set Non_ADS_Alpha(Non_ADS_Alpha:double=0) -> (Output_Get:double)
   Set ADS_Alpha(execute=@n2.else, ADS_Alpha=0)
   > Set Non_ADS_Alpha(Non_ADS_Alpha=1)
  ]}
]
variables [
 ADS_Alpha {type=number}
 Non_ADS_Alpha {type=number}
 skyDivingAlpha {type=number}
]
