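# Physics asset for the head part of the male medium Famine player body cosmetic, simulating its dangling cage and lock bones.
# Capsule bodies cover the head, dyn_cage_01 and dyn_lock_02, chained by two constraints.
# The cage has very tight swing and twist limits against the head, while the lock has wide swing limits on the cage with twist locked.
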PhysicsAsset /BRCosmetics/Characters/Player/Male/Medium/Bodies/M_MED_Famine/Meshes/Parts/Male_Medium_Famine_Head_Physics
bodies {
 dyn_cage_01 [{shapeType=Capsule, length=5, radius=5, rotation=<rot 16.2771 -90 18.7546>}]
 dyn_lock_02 [{shapeType=Capsule, length=5, radius=2, center=<5.4934 -0.0001 0>, rotation=<rot 90 0 0>}]
 head [{shapeType=Capsule, length=5, radius=12, center=<7.012 0.6116 -0.0001>, rotation=<rot 90 0 0>}]
}
constraints [
 {bone1Name=dyn_cage_01, bone2Name=head, swing1LimitDegrees=2, swing2LimitDegrees=2, twistLimitDegrees=2}
 {bone1Name=dyn_lock_02, bone2Name=dyn_cage_01, swing1LimitDegrees=75, swing2LimitDegrees=75, twistLimitDegrees=45, twistMotion=Locked}
]
